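# Niagara intro effect for the AquaPeony emote that kicks up sand puffs from the character's feet.
# Two timed sprite bursts spawn at the left and right feet, with timing and offsets varying by front end versus gameplay.
# Owner velocity switches timing, gravity curves, world/local blending and drift, while Intensity scales launch speed and time of day dims color.
# Airglow, smoke and sparkle emitters appear without modules or renderers.
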NiagaraSystem /BRCosmetics/Animation/Game/MainPlayer/Emotes/AquaPeony/FX/NS_AquaPeony_Intro
dependencies {
 usedDataInterfaces [&NiagaraDataInterfaceVectorCurve, &NiagaraDataInterfaceCurve, &NiagaraDataInterfaceSkeletalMesh]
 usedDynamicInputs []
 usedModules [
  &SystemState
  &BlendSpaceHelper
  &self:SystemSpawnScript.SystemScriptSource.SystemScriptGraph.NiagaraNodeAssignment_0.SetVariables_#1
  &self:SystemSpawnScript.SystemScriptSource.SystemScriptGraph.NiagaraNodeAssignment_1.SetVariables_#2
 ]
}
emitters {
 AirglowLWE {rendererClasses [], renderers []}
 Sand {
  emitterUpdateScript [
   &EmitterState()
   &SpawnBurst_Instantaneous(
    Spawn Count=2
    Spawn Time=&FN_FrontEndFloatModifier(
     Gameplay=&MakeCustomFloatFromBool(
      Bool=&SetBoolByFloatComparison(A=&VectorLength(Vector=$User.OwnerLocalVelocity), B=20)
      True Float=0.83
      False Float=0.43
     )
     Front End=0.83
    )
    Use Spawn Probability=false
    UseLoopCountLimit=false
   ) {module=SpawnBurst_Instantaneous001}
   &SpawnBurst_Instantaneous(
    Spawn Count=2
    Spawn Group=1
    Spawn Time=&FN_FrontEndFloatModifier(
     Gameplay=&MakeCustomFloatFromBool(
      Bool=&SetBoolByFloatComparison(A=&VectorLength(Vector=$User.OwnerLocalVelocity), B=20)
      True Float=1.5
      False Float=1.03
     )
     Front End=1.5
    )
    Use Spawn Probability=false
    UseLoopCountLimit=false
   ) {module=SpawnBurst_Instantaneous002}
  ]
  particleSpawnScript [
   &InitializeParticle(
    Color Maximum=<rgba 0.674 0.4176 0.2984 1.5>
    Color Minimum=&Multiply_LinearColorByFloat(
     Linear Color=<rgba 0.8406 0.4865 0.2881 1.5>
     Float=&FNTimeOfDayScalar(Night Scale=0.5, Morning Scale=0.6, Evening Scale=0.6)
    )
    Color Mode=::NewEnumerator2
    Sprite Size Mode=::NewEnumerator1
    Uniform Sprite Size Max=100
    Uniform Sprite Size Min=50
    UsePositionOffset=false
    Write Material Random=true
   )
   &ShapeLocation(
    Handle Radius=9
    Large Radius=11
    MaskSpawnGroup=false
    Non Uniform Scale=<1.7 4 0.85>
    Override Seed=false
    Shape Primitive=::NewEnumerator3
    UseOwnerScale=true
   ) {module=ShapeLocation001}
   &AddVelocity(
    Cone Angle=30
    Cone Axis=<0 -1 0>
    Module.UseConeFalloff=false
    Module.UseVelDistribution=false
    Override Seed=false
    Velocity Mode=::NewEnumerator2
    Velocity Speed=&Multiply_Float(A=&RandomRangeFloat(Minimum=10, Maximum=15), B=$User.Intensity)
   ) {module=AddVelocity001}
   &AddVelocity(
    Module.Use Falloff Distance=false
    Origin Offset=<-10 0 0>
    Override Seed=false
    Velocity Mode=::NewEnumerator1
    Velocity Speed=&Multiply_Float(A=&RandomRangeFloat(Minimum=15, Maximum=20), B=$User.Intensity)
   )
   &CameraOffset(Camera Offset Amount=&RandomRangeFloat(Minimum=5, Maximum=10))
   &SkeletalMeshLocation(
    MaskSpawnGroup=true
    Mesh Sampling Type=::NewEnumerator5
    Sampled Position Offset=&FN_FrontEndVector3Modifier(Gameplay=<0 12 -5>, Front End=<0 -6 -3>)
    Sampled Position Offset Coordinate Space:ENiagaraSkelMeshTransforms=::NewEnumerator2
    Skeletal Mesh=$System.Left Foot
    Socket Sampling Mode=::NewEnumerator5
   )
   &SkeletalMeshLocation(
    MaskSpawnGroup=true
    Mesh Sampling Type=::NewEnumerator5
    Sampled Position Offset=&FN_FrontEndVector3Modifier(Gameplay=<0 12 -5>, Front End=<0 -6 -3>)
    Sampled Position Offset Coordinate Space:ENiagaraSkelMeshTransforms=::NewEnumerator2
    Skeletal Mesh=$System.Right Foot
    Socket Sampling Mode=::NewEnumerator5
    Spawn Group Mask=1
   ) {module=SkeletalMeshLocation001}
  ]
  particleUpdateScript [
   &ParticleState()
   &GravityForce(Gravity=&MakeCustomVectorFromBool(
     Bool=&SetBoolByFloatComparison(A=&VectorLength(Vector=$User.OwnerLocalVelocity), B=20)
     True Vector=&VectorFromCurve(VectorCurve=DataInterface() {propertyValues {
        XCurve {keys [<at 0 0>,<at 1 0>]}
        YCurve {keys [
          {arriveTangent=-21.7461, interpMode=RCIM_Cubic, leaveTangent=-21.7461, tangentMode=RCTM_User, time=0, value=-10}
          {interpMode=RCIM_Cubic, tangentMode=RCTM_User, time=1, value=-20}
         ]}
        ZCurve {keys [
          {arriveTangent=578.5428, interpMode=RCIM_Cubic, leaveTangent=578.5428, tangentMode=RCTM_User, time=0, value=0}
          {arriveTangent=-6573.3198, interpMode=RCIM_Cubic, leaveTangent=-2867.7527, tangentMode=RCTM_Break, time=0.2, value=0}
          {arriveTangent=54.3671, interpMode=RCIM_Cubic, leaveTangent=54.3671, tangentMode=RCTM_User, time=1, value=-600}
         ]}
       }})
     False Vector=&VectorFromCurve(VectorCurve=DataInterface() {propertyValues {
        XCurve {keys [<at 0 0>,<at 1 0>]}
        YCurve {keys [
          {arriveTangent=-21.7461, interpMode=RCIM_Cubic, leaveTangent=-21.7461, tangentMode=RCTM_User, time=0, value=-10}
          {interpMode=RCIM_Cubic, tangentMode=RCTM_User, time=1, value=-20}
         ]}
        ZCurve {keys [
          {arriveTangent=800.1272, interpMode=RCIM_Cubic, leaveTangent=800.1272, tangentMode=RCTM_User, time=0, value=0}
          {arriveTangent=-3507.2607, interpMode=RCIM_Cubic, leaveTangent=-2867.7527, tangentMode=RCTM_Break, time=0.2, value=0}
          {arriveTangent=54.3671, interpMode=RCIM_Cubic, leaveTangent=54.3671, tangentMode=RCTM_User, time=1, value=-600}
         ]}
       }})
    ))
   &BlendSpace("Position Blend (World <-> Local)":NiagaraFloat=&MakeCustomFloatFromBool(
     Bool=&SetBoolByFloatComparison(A=&VectorLength(Vector=$Engine.Owner.Velocity), B=301)
     True Float=1
     False Float=&FloatFromCurve(FloatCurve=DataInterface() {propertyValues {Curve {
         defaultValue=0
         keys [
          {arriveTangent=-7.3261, interpMode=RCIM_Cubic, leaveTangent=-7.3261, tangentMode=RCTM_User, time=0, value=1}
          {arriveTangent=-1.5885, interpMode=RCIM_Cubic, leaveTangent=-1.5885, tangentMode=RCTM_User, time=0.2, value=0.2}
          <at 0.5 0 RCIM_Cubic>
         ]
        }}})
     Translation Contribution:NiagaraFloat=1
     Rotation Contribution:NiagaraFloat=1
    )) {module=BlendSpace001}
   &Drag(Drag=2, Use Linear Drag=true, Use Rotational Drag=false)
   &SolveForcesAndVelocity(Clamp Velocity=false, Limit Acceleration=false) {module=SolveForcesAndVelocity001}
   &ScaleSpriteSize(
    Uniform Curve Scale=0.7
    Uniform Curve Sprite Scale=DataInterface() {propertyValues {Curve {keys [
        {arriveTangent=7.4995, interpMode=RCIM_Cubic, leaveTangent=7.4995, tangentMode=RCTM_User, time=0, value=0.15}
        {arriveTangentWeight=0.1312, interpMode=RCIM_Cubic, leaveTangentWeight=0.3062, tangentMode=RCTM_User, time=0.3, value=1}
        {arriveTangent=0.8571, interpMode=RCIM_Cubic, leaveTangent=0.8571, time=0.8, value=1}
        {arriveTangent=2.6672, arriveTangentWeight=0.266, interpMode=RCIM_Cubic, leaveTangent=2.6672, tangentMode=RCTM_User, time=1, value=1.6}
       ]}}}
   )
   &ScaleColor(
    Scale Alpha=&FloatFromCurve(
     FloatCurve=DataInterface() {propertyValues {Curve {keys [
         <at 0 0.3 RCIM_Cubic>
         {arriveTangent=9.9601, interpMode=RCIM_Cubic, leaveTangent=9.9601, tangentMode=RCTM_User, time=0.02, value=0.6}
         {arriveTangent=-2.7573, interpMode=RCIM_Cubic, leaveTangent=-1.7516, tangentMode=RCTM_Break, time=0.1, value=0.6}
         {arriveTangentWeight=0.2045, interpMode=RCIM_Cubic, tangentMode=RCTM_User, time=1, value=0}
        ]}}}
     Scale Curve=&FN_FrontEndFloatModifier(Gameplay=0.7, Front End=0.4)
    )
    ScaleA=true
    ScaleRGB=true
   )
   &SolveForcesAndVelocity(Clamp Velocity=false, Limit Acceleration=false)
   /SetVariables(Particles.Position=&AddVectorToPosition(
     Position=$Particles.Position
     Vector=&MakeCustomVectorFromBool(
      Bool=&SetBoolByFloatComparison(A=&VectorLength(Vector=$User.OwnerLocalVelocity), B=20)
      True Vector=&VectorFromCurve(
       VectorCurve=DataInterface() {propertyValues {XCurve {keys [<at 0 0>,<at 1 0>]}, YCurve {keys [<at 0 -1>,<at 1 0>]}, ZCurve {keys [<at 0 0>,<at 1 0>]}}}
       Scale Curve=<0.2 0.22 0.2>
      )
      False Vector=<0 0 0>
     )
    )) {isSetParameters=true}
  ]
  rendererClasses [&NiagaraSpriteRendererProperties]
  renderers [{rendererClass=&NiagaraSpriteRendererProperties}]
 }
 SmokeLWE {rendererClasses [], renderers []}
 SparklesLWE {rendererClasses [], renderers []}
}
summary {
 systemName=NS_AquaPeony_Intro
 emitters [
  {emitterName=SparklesLWE, rendererClasses []}
  {emitterName=AirglowLWE, rendererClasses []}
  {emitterName=SmokeLWE, rendererClasses []}
  {emitterName=Sand, rendererClasses [&NiagaraSpriteRendererProperties]}
 ]
 userVariables [
  User.bIsFrontEnd {type=NiagaraBool}
  User.bIsFrontEndPreview {type=NiagaraBool}
  User.Intensity {type=NiagaraFloat, defaultValue=0.4, description=""}
  User.NormalizedNotifyProgress {type=NiagaraFloat, defaultValue=0, description=""}
  User.OwnerLocalVelocity {type=Vector3f, description="", defaultValue {struct=Vector3f, value=<0 0 0>}}
 ]
}
