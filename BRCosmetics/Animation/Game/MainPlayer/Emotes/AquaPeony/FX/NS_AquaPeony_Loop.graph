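# Niagara system for the looping effect of the AquaPeony emote, layering warm flickering light glints with pale blue mist, sparkle mist and droplets.
# Emitters pack several looks into one via spawn groups, visibility tags, masked inputs and multiple sprite renderers.
# Glint and mist positions blend between world and local space using owner speed and particle-age curves.
# Brightness adapts to time of day and front-end context; Mist_Condensed is disabled.
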NiagaraSystem /BRCosmetics/Animation/Game/MainPlayer/Emotes/AquaPeony/FX/NS_AquaPeony_Loop
dependencies {
 usedDataInterfaces [
  &NiagaraDataInterfaceCurve
  &NiagaraDataInterfaceSpriteRendererInfo
  &NiagaraDataInterfaceVectorCurve
  &NiagaraDataInterfaceVector2DCurve
  &NiagaraDataInterfaceColorCurve
 ]
 usedDynamicInputs [/Niagara/DynamicInputs/Helpers/MaskVector2d, &FNTimeOfDayColor]
 usedModules [
  &self:SystemSpawnScript.SystemScriptSource.SystemScriptGraph.NiagaraNodeAssignment_0.SetVariables_#1
  &self:SystemSpawnScript.SystemScriptSource.SystemScriptGraph.NiagaraNodeAssignment_1.SetVariables_#2
  &SystemState
  &BlendSpaceHelper
 ]
}
emitters {
 Glow_Fake_Light {rendererClasses [], renderers []}
 Light_Glints001 {
  emitterUpdateScript [
   &EmitterState(Life Cycle Mode=::NewEnumerator1, Loop Duration=1, UseLoopDelay=false)
   &SpawnRate(SpawnRate=0.6, Use Spawn Probability=false)
   &SpawnRate(SpawnGroup=1, SpawnRate=4, Use Spawn Probability=false) {module=SpawnRate001}
   &SpawnRate(SpawnGroup=3, SpawnRate=0.7, Use Spawn Probability=false) {module=SpawnRate003}
  ]
  particleSpawnScript [
   /SetVariables(Particles.VisibilityTag=&SpawnGroup()) {isSetParameters=true}
   &InitializeParticle(
    Color Mode=::NewEnumerator0
    Lifetime Max=6
    Lifetime Min=3
    Lifetime Mode=::NewEnumerator1
    Sprite Rotation Mode=::NewEnumerator2
    Sprite Size Mode=::NewEnumerator3
    Uniform Sprite Size=&RandomRangeFloat(
     Minimum=&MaskFloat(Mask Source=0, Mask Parameter=?, Mask=0, Masked Float=2, Passthrough Float=3)
     Maximum=&MaskFloat(Mask Source=0, Mask Parameter=?, Mask=0, Masked Float=5.5, Passthrough Float=42)
    )
    UsePositionOffset=false
    Write Material Random=true
   )
   &ShapeLocation(
    Box Size=<100 100 90>
    MaskSpawnGroup=false
    Offset=<0 0 90>
    Offset Mode=::NewEnumerator0
    Override Seed=false
    Shape Primitive=::NewEnumerator2
    Surface Only=false
    UseOwnerScale=true
   )
   /SetVariables(Particles.Flare SpriteSize:Vector2f=&Vector2DFromFloat(Value=&RandomRangeFloat(Minimum=1.8, Maximum=4.8))) {isSetParameters=true}
   &CameraOffset(Camera Offset Amount=&MaskFloat(
     Mask Source=0
     Mask Parameter=?
     Mask=0
     Masked Float=&RandomRangeFloat(Minimum=12, Maximum=20)
     Passthrough Float=&MaskFloat(
      Mask Source=0
      Mask Parameter=?
      Mask=1
      Masked Float=&RandomRangeFloat(Minimum=10, Maximum=17)
      Passthrough Float=&MaskFloat(
       Mask Source=0
       Mask Parameter=?
       Mask=2
       Masked Float=&RandomRangeFloat(Minimum=10, Maximum=17)
       Passthrough Float=&RandomRangeFloat(Minimum=20, Maximum=27)
      )
     )
    ))
  ]
  particleUpdateScript [
   &ParticleState()
   &ScaleSpriteSize(
    Scale Sprite Size Mode=::NewEnumerator0
    Uniform Scale Factor=&MaskFloat(
     Mask Source=0
     Mask Parameter=?
     Mask=0
     Masked Float=&FloatFromCurve(FloatCurve=DataInterface() {propertyValues {Curve {keys [
          {arriveTangent=7.8653, interpMode=RCIM_Cubic, leaveTangent=7.8653, tangentMode=RCTM_User, time=0.0033, value=0.0151}
          {arriveTangent=0.8866, interpMode=RCIM_Cubic, leaveTangent=0.8866, tangentMode=RCTM_User, time=0.2773, value=0.783}
          {arriveTangent=12.6874, interpMode=RCIM_Cubic, leaveTangent=-13.0744, tangentMode=RCTM_Break, time=0.4458, value=1.3521}
          {arriveTangent=-0.8088, interpMode=RCIM_Cubic, leaveTangent=-0.8088, tangentMode=RCTM_User, time=0.6094, value=0.8056}
          {arriveTangent=-5.9884, interpMode=RCIM_Cubic, leaveTangent=-5.9884, tangentMode=RCTM_User, time=1, value=0}
         ]}}})
     Passthrough Float=&FloatFromCurve(FloatCurve=DataInterface() {propertyValues {Curve {keys [
          <at 0 0 RCIM_Cubic>
          {arriveTangent=3.1939, interpMode=RCIM_Cubic, leaveTangent=2.7137, tangentMode=RCTM_Break, time=0.45, value=1}
          {arriveTangent=-1.7556, interpMode=RCIM_Cubic, leaveTangent=-2.8595, tangentMode=RCTM_Break, time=0.55, value=1}
          {arriveTangent=0.4826, interpMode=RCIM_Cubic, leaveTangent=0.4826, tangentMode=RCTM_User, time=1, value=0}
         ]}}})
    )
   )
   &BlendSpace("Position Blend (World <-> Local)":NiagaraFloat=&MakeCustomFloatFromBool(
     Bool=&SetBoolByFloatComparison(A=&VectorLength(Vector=$Engine.Owner.Velocity), B=301)
     True Float=1
     False Float=&FloatFromCurve(FloatCurve=DataInterface() {propertyValues {Curve {
         defaultValue=0
         keys [
          {arriveTangent=-6.7037, interpMode=RCIM_Cubic, leaveTangent=-6.7037, tangentMode=RCTM_User, time=0, value=1}
          {interpMode=RCIM_Cubic, tangentMode=RCTM_User, time=0.2, value=0.5}
         ]
        }}})
     Translation Contribution:NiagaraFloat=1
     Rotation Contribution:NiagaraFloat=0
    )) {module=BlendSpace001}
   &ScaleColor(
    Linear Color Curve=DataInterface() {propertyValues {
      AlphaCurve {keys [<at 0 0>,<at 0.0436 0.02>,<at 0.1025 0.1>,<at 0.3189 0.2>,<at 0.4939 0.55>,<at 0.7632 0.2>,<at 0.9046 0.1>,<at 0.9572 0.02>,<at 1 0>]}
      BlueCurve {keys [<at 0 0.2133>,<at 0.434 0.1223>,<at 0.4921 0.5919>,<at 0.5474 0.1221>,<at 0.7759 0.1849>,<at 0.9935 0.4573>]}
      GreenCurve {keys [<at 0 0.2457>,<at 0.434 0.3967>,<at 0.4921 3.8934>,<at 0.5474 0.3968>,<at 0.7759 0.1063>,<at 0.9935 0.0881>]}
      RedCurve {keys [<at 0 1>,<at 0.434 0.9448>,<at 0.4921 10>,<at 0.5474 0.9473>,<at 0.7759 0.7448>,<at 0.9935 0.287>]}
     }}
    Scale Mode=::NewEnumerator2
   )
   /SetVariables(Particles.Motes Dynamic Param:Vector4f=&MakeVector4(
     X=&Waveform(
      "Waveform [1]"=::NewEnumerator0
      "[1] Frequency"=&RandomRangeFloat(Minimum=2, Maximum=4)
      "[1] Amplitude Min/Max"=<0 0.7>
     )
     Y=0
     Z=0
     W=0
    )) {isSetParameters=true}
   /SetVariables(Particles.Glitter Color:LinearColor=&Multiply_LinearColorByFloat(
     Linear Color=<rgba 1 0.8958 0.28 0.2>
     Float=&FloatFromCurve(
      FloatCurve=DataInterface() {propertyValues {Curve {keys [
          {arriveTangent=-0.012, interpMode=RCIM_Cubic, leaveTangent=-0.012, tangentMode=RCTM_User, time=0, value=0}
          {arriveTangent=1.1429, interpMode=RCIM_Cubic, leaveTangent=1.1429, time=0.2, value=1}
          {interpMode=RCIM_Cubic, tangentMode=RCTM_User, time=0.35, value=0.4}
          {interpMode=RCIM_Cubic, tangentMode=RCTM_User, time=0.5, value=0.6}
          {arriveTangent=-0.9375, interpMode=RCIM_Cubic, leaveTangent=-0.9375, time=0.68, value=0.15}
          {arriveTangent=-0.4688, interpMode=RCIM_Cubic, leaveTangent=-0.4688, time=0.82, value=0.3}
          {interpMode=RCIM_Cubic, tangentMode=RCTM_Break, time=1, value=0}
          {interpMode=RCIM_Cubic, leaveTangent=-1.75, time=5, value=0.2}
         ]}}}
      Scale Curve=&Multiply_Float(A=0.85, B=&FNTimeOfDayScalar(Night Scale=0.7, Morning Scale=0.8, Evening Scale=0.8))
     )
    )) {isSetParameters=true}
   &CurlNoiseForce(
    Bias Noise=false
    Noise Frequency=20
    Noise Strength=&FloatFromCurve(FloatCurve=DataInterface() {propertyValues {Curve {keys [{arriveTangent=-158.0912, interpMode=RCIM_Cubic, leaveTangent=-158.0912, tangentMode=RCTM_User, time=0, value=50}, <at 1 0>]}}})
    Pan Noise=false
   )
   &SolveForcesAndVelocity(Clamp Velocity=false, Limit Acceleration=false)
   /SetVariables(Particles.Flare Color:LinearColor=&MakeLinearColorFromVectorAndFloat(
     "Vector (RGB)"=&MakeVectorFromLinearColorRGB(Color=&MultiplyLinearColor(
       Linear Color=&RandomRangeLinearColor(Minimum=<rgba 2 0.5294 0.0936 0.8>, Maximum=<rgba 10 5.3571 0.9476 0.5>)
       Scale Factor=/Niagara/DynamicInputs/UniformRange/V2/RandomRangeVector4(
        Minimum:Vector4f=<0.8 0.9 0.9 0.8>
        Maximum:Vector4f=<1.6 1 1.1 1>
        Evaluation Type:ENiagaraRandomnessEvaluation=::NewEnumerator0
        Recalculate Random Each Loop:NiagaraBool=true
        Randomness Mode:ENiagaraRandomnessMode=::NewEnumerator0
        Random Seed:NiagaraInt32=0
        Fixed Random Seed:NiagaraBool=false
       )
      ))
     "Float (Alpha)"=&Multiply_Float(
      A=&FloatFromCurve(FloatCurve=DataInterface() {propertyValues {Curve {keys [
           <at -0.005 0 RCIM_Cubic>
           {arriveTangent=35.0473, interpMode=RCIM_Cubic, leaveTangent=35.0473, time=0.0204, value=1.0351}
           {arriveTangent=8.4622, interpMode=RCIM_Cubic, leaveTangent=8.4622, tangentMode=RCTM_User, time=0.0282, value=1.1627}
           {interpMode=RCIM_Cubic, tangentMode=RCTM_User, time=0.0515, value=0}
           {interpMode=RCIM_Cubic, tangentMode=RCTM_User, time=0.0702, value=1.0453}
           {interpMode=RCIM_Cubic, tangentMode=RCTM_User, time=0.1465, value=0.8256}
           {interpMode=RCIM_Cubic, tangentMode=RCTM_User, time=0.1468, value=-0.0075}
           {interpMode=RCIM_Cubic, tangentMode=RCTM_User, time=0.1611, value=0}
           {interpMode=RCIM_Cubic, tangentMode=RCTM_User, time=0.1759, value=0.6661}
           {interpMode=RCIM_Cubic, tangentMode=RCTM_User, time=0.1958, value=0}
           {interpMode=RCIM_Cubic, tangentMode=RCTM_User, time=0.2128, value=0}
           {interpMode=RCIM_Cubic, tangentMode=RCTM_User, time=0.2197, value=0.5011}
           {interpMode=RCIM_Cubic, tangentMode=RCTM_User, time=0.241, value=0.5011}
           {interpMode=RCIM_Cubic, tangentMode=RCTM_User, time=0.2491, value=0}
           {interpMode=RCIM_Cubic, tangentMode=RCTM_User, time=0.275, value=0}
           {interpMode=RCIM_Cubic, tangentMode=RCTM_User, time=0.296, value=0.391}
           {interpMode=RCIM_Cubic, tangentMode=RCTM_User, time=0.3287, value=0.2412}
           {interpMode=RCIM_Cubic, tangentMode=RCTM_User, time=0.3401, value=0}
           {interpMode=RCIM_Cubic, tangentMode=RCTM_User, time=0.4038, value=0}
           {interpMode=RCIM_Cubic, tangentMode=RCTM_User, time=0.4083, value=0.2622}
           {interpMode=RCIM_Cubic, tangentMode=RCTM_User, time=0.6242, value=0.0894}
           {interpMode=RCIM_Cubic, tangentMode=RCTM_User, time=0.6346, value=0}
           {interpMode=RCIM_Cubic, tangentMode=RCTM_User, time=0.6716, value=0}
           {interpMode=RCIM_Cubic, tangentMode=RCTM_User, time=0.6823, value=0.1536}
           {interpMode=RCIM_Cubic, tangentMode=RCTM_User, time=0.6941, value=0}
           {interpMode=RCIM_Cubic, tangentMode=RCTM_User, time=0.7334, value=0}
           {interpMode=RCIM_Cubic, tangentMode=RCTM_User, time=0.7497, value=0.1144}
           {interpMode=RCIM_Cubic, tangentMode=RCTM_User, time=0.7986, value=0.0343}
           {interpMode=RCIM_Cubic, tangentMode=RCTM_User, time=0.8136, value=0}
           {interpMode=RCIM_Cubic, tangentMode=RCTM_User, time=0.848, value=0}
           {interpMode=RCIM_Cubic, tangentMode=RCTM_User, time=0.8801, value=0.0578}
           <at 0.8975 0 RCIM_Cubic>
          ]}}})
      B=&FloatFromCurve(FloatCurve=DataInterface() {propertyValues {Curve {keys [
           {interpMode=RCIM_Cubic, tangentMode=RCTM_User, time=0.0082, value=0}
           {interpMode=RCIM_Cubic, tangentMode=RCTM_User, time=0.1922, value=1}
           {interpMode=RCIM_Cubic, tangentMode=RCTM_User, time=0.4736, value=0.5019}
           {interpMode=RCIM_Cubic, tangentMode=RCTM_User, time=0.7676, value=1}
           {interpMode=RCIM_Cubic, tangentMode=RCTM_User, time=1, value=0}
          ]}}})
     )
    )) {isSetParameters=true}
   /SetVariables(Particles.Update Flare SpriteSize:Vector2f=&Multiply_Vector2D(
     A=$Particles.Flare SpriteSize
     B=&Vector2DFromFloat(Value=&Multiply_Float(
       A=&Waveform(
        "Waveform [1]"=::NewEnumerator0
        "[1] Frequency"=&FloatFromCurve(FloatCurve=DataInterface() {propertyValues {Curve {keys [
             {arriveTangent=3.0622, interpMode=RCIM_Cubic, leaveTangent=3.0622, tangentMode=RCTM_User, time=0, value=0.007}
             {arriveTangent=0.1346, interpMode=RCIM_Cubic, leaveTangent=0.1346, tangentMode=RCTM_User, time=0.15, value=0.002}
             {arriveTangent=-0.1241, interpMode=RCIM_Cubic, leaveTangent=-0.1241, tangentMode=RCTM_User, time=0.85, value=0.002}
             {arriveTangent=-2.948, interpMode=RCIM_Cubic, leaveTangent=-2.948, tangentMode=RCTM_User, time=1, value=0.007}
            ]}}})
        "[1] Amplitude Min/Max"=<0.5 1>
        "[1] Bias"=1
       )
       B=&FloatFromCurve(FloatCurve=DataInterface() {propertyValues {Curve {keys [
            {interpMode=RCIM_Cubic, tangentMode=RCTM_User, time=0.0418, value=1}
            {interpMode=RCIM_Cubic, tangentMode=RCTM_User, time=0.0515, value=0.4}
            {interpMode=RCIM_Cubic, tangentMode=RCTM_User, time=0.0579, value=1}
            {interpMode=RCIM_Cubic, tangentMode=RCTM_User, time=0.1411, value=1}
            {interpMode=RCIM_Cubic, tangentMode=RCTM_User, time=0.1468, value=0.4}
            {interpMode=RCIM_Cubic, tangentMode=RCTM_User, time=0.1611, value=0.4}
            {interpMode=RCIM_Cubic, tangentMode=RCTM_User, time=0.1749, value=1}
            {interpMode=RCIM_Cubic, tangentMode=RCTM_User, time=0.1958, value=0.4}
            {interpMode=RCIM_Cubic, tangentMode=RCTM_User, time=0.2128, value=0.4}
            {interpMode=RCIM_Cubic, tangentMode=RCTM_User, time=0.2166, value=1}
            {interpMode=RCIM_Cubic, tangentMode=RCTM_User, time=0.2379, value=1}
            {interpMode=RCIM_Cubic, tangentMode=RCTM_User, time=0.2491, value=0.4}
            {interpMode=RCIM_Cubic, tangentMode=RCTM_User, time=0.275, value=0.4}
            {interpMode=RCIM_Cubic, tangentMode=RCTM_User, time=0.277, value=1}
            {interpMode=RCIM_Cubic, tangentMode=RCTM_User, time=0.3332, value=1}
            {interpMode=RCIM_Cubic, tangentMode=RCTM_User, time=0.3401, value=0.4}
            {interpMode=RCIM_Cubic, tangentMode=RCTM_User, time=0.4038, value=0.4}
            {interpMode=RCIM_Cubic, tangentMode=RCTM_User, time=0.4111, value=1}
            {interpMode=RCIM_Cubic, tangentMode=RCTM_User, time=0.6269, value=1}
            {interpMode=RCIM_Cubic, tangentMode=RCTM_User, time=0.6346, value=0.4}
            {interpMode=RCIM_Cubic, tangentMode=RCTM_User, time=0.6716, value=0.4}
            {interpMode=RCIM_Cubic, tangentMode=RCTM_User, time=0.6808, value=1}
            {interpMode=RCIM_Cubic, tangentMode=RCTM_User, time=0.6941, value=0.4}
            {interpMode=RCIM_Cubic, tangentMode=RCTM_User, time=0.7334, value=0.4}
            {interpMode=RCIM_Cubic, tangentMode=RCTM_User, time=0.7439, value=1}
            {interpMode=RCIM_Cubic, tangentMode=RCTM_User, time=0.7928, value=1}
            {interpMode=RCIM_Cubic, tangentMode=RCTM_User, time=0.8136, value=0.4}
            {interpMode=RCIM_Cubic, tangentMode=RCTM_User, time=0.848, value=0.4}
            {interpMode=RCIM_Cubic, tangentMode=RCTM_User, time=0.8661, value=1}
           ]}}})
      ))
    )) {isSetParameters=true}
  ]
  rendererClasses [&NiagaraSpriteRendererProperties]
  renderers [
   {rendererClass=&NiagaraSpriteRendererProperties}
   {rendererClass=&NiagaraSpriteRendererProperties, rendererIndex=1}
   {rendererClass=&NiagaraSpriteRendererProperties, rendererIndex=2}
  ]
 }
 Mist002 {
  emitterUpdateScript [
   &EmitterState(Life Cycle Mode=::NewEnumerator1, Loop Duration=1, UseLoopDelay=false)
   &SpawnRate(SpawnRate=3, Use Spawn Probability=false) {module=SpawnRate001}
   &SpawnRate(SpawnGroup=1, SpawnRate=1.9, Use Spawn Probability=false)
   &SpawnRate(SpawnGroup=2, SpawnRate=3.5, Use Spawn Probability=false) {module=SpawnRate002}
  ]
  particleSpawnScript [
   /SetVariables(Particles.VisibilityTag=&SpawnGroup()) {isSetParameters=true}
   &InitializeParticle(
    Color Maximum=&Multiply_LinearColorByFloat(
     Linear Color=<rgba 0.4498 0.8309 1 0.02>
     Float=&FNTimeOfDayScalar(Night Scale=0.5, Morning Scale=0.65, Evening Scale=0.65)
    )
    Color Minimum=&Multiply_LinearColorByFloat(
     Linear Color=<rgba 0.7353 0.9636 1 0.02>
     Float=&FNTimeOfDayScalar(Night Scale=0.5, Morning Scale=0.65, Evening Scale=0.65)
    )
    Color Mode=::NewEnumerator2
    Lifetime=&MaskFloat(
     Mask Source=0
     Mask Parameter=?
     Mask=0
     Masked Float=&RandomRangeFloat(Minimum=2, Maximum=3)
     Passthrough Float=&MaskFloat(
      Mask Source=0
      Mask Parameter=?
      Mask=1
      Masked Float=&RandomRangeFloat(Minimum=0.6, Maximum=2)
      Passthrough Float=&RandomRangeFloat(Minimum=1, Maximum=2.5)
     )
    )
    Sprite Rotation Mode=::NewEnumerator2
    Sprite Size Mode=::NewEnumerator1
    Uniform Sprite Size Max=&MaskFloat(
     Mask Source=0
     Mask Parameter=?
     Mask=0
     Masked Float=140
     Passthrough Float=&MaskFloat(Mask Source=0, Mask Parameter=?, Mask=1, Masked Float=140, Passthrough Float=2)
    )
    Uniform Sprite Size Min=&MaskFloat(
     Mask Source=0
     Mask Parameter=?
     Mask=0
     Masked Float=70
     Passthrough Float=&MaskFloat(Mask Source=0, Mask Parameter=?, Mask=1, Masked Float=70, Passthrough Float=0.5)
    )
    UsePositionOffset=false
    Write Material Random=true
   )
   &ShapeLocation(
    Box Size=&MaskVector(
     Mask Source=0
     Mask Parameter=?
     Mask=0
     Masked Vector=<80 100 120>
     Passthrough Vector=&MaskVector(Mask Source=0, Mask Parameter=?, Mask=1, Masked Vector=<80 100 50>, Passthrough Vector=<80 100 100>)
    )
    MaskSpawnGroup=false
    Offset=&MaskVector(
     Mask Source=0
     Mask Parameter=?
     Mask=0
     Masked Vector=<0 35 100>
     Passthrough Vector=&MaskVector(Mask Source=0, Mask Parameter=?, Mask=1, Masked Vector=<0 55 45>, Passthrough Vector=<0 65 90>)
    )
    Offset Mode=::NewEnumerator0
    Override Seed=false
    Shape Primitive=::NewEnumerator2
    Surface Only=false
    UseOwnerScale=true
   )
  ]
  particleUpdateScript [
   &ParticleState()
   &ScaleSpriteSize(Uniform Curve Sprite Scale=DataInterface() {propertyValues {Curve {keys [
        {arriveTangent=3.3351, interpMode=RCIM_Cubic, leaveTangent=3.3351, tangentMode=RCTM_User, time=0, value=0}
        {arriveTangent=0.0592, interpMode=RCIM_Cubic, leaveTangent=0.0592, tangentMode=RCTM_User, time=0.7, value=0.65}
        {arriveTangent=2.1747, interpMode=RCIM_Cubic, leaveTangent=2.1747, tangentMode=RCTM_User, time=1, value=1}
       ]}}})
   &SubUVAnimation(
    Define SubUV Setup Manually:ENiagara_AutomaticManual=::NewEnumerator1
    Is SubUV Blending Enabled On Renderer=true
    Output Alternate Renderer Binding=true
    SubUV Lookup Index Scale:NiagaraFloat=&FloatFromCurve(FloatCurve=DataInterface() {propertyValues {Curve {keys [
         {arriveTangent=0.8228, interpMode=RCIM_Cubic, leaveTangent=0.8228, tangentMode=RCTM_User, time=-0.0072, value=0}
         {arriveTangent=1.6719, interpMode=RCIM_Cubic, leaveTangent=1.6719, tangentMode=RCTM_User, time=1, value=1}
        ]}}})
    UseEndFrame=false
    UseStartFrame=false
   ) {module=SubUVAnimation001}
   &ScaleColor(
    Scale Alpha=&FloatFromCurve(
     FloatCurve=DataInterface() {propertyValues {Curve {keys [
         <at 0 0 RCIM_Cubic>
         {arriveTangent=6.0453, interpMode=RCIM_Cubic, leaveTangent=6.0453, tangentMode=RCTM_User, time=0.4, value=1}
         {arriveTangent=-5.3053, interpMode=RCIM_Cubic, leaveTangent=-5.3053, tangentMode=RCTM_User, time=0.5, value=1}
         <at 1 0 RCIM_Cubic>
        ]}}}
     Scale Curve=&FN_FrontEndFloatModifier(Gameplay=1, Front End=0.5)
    )
    ScaleA=true
    ScaleRGB=true
   )
   &CurlNoiseForce(
    Bias Noise=false
    Noise Frequency=40
    Noise Strength=&MaskFloat(Mask Source=0, Mask Parameter=?, Mask=0, Masked Float=0, Passthrough Float=30)
    Pan Noise=false
   )
   &BlendSpace("Position Blend (World <-> Local)":NiagaraFloat=&MakeCustomFloatFromBool(
     Bool=&SetBoolByFloatComparison(A=&VectorLength(Vector=$Engine.Owner.Velocity), B=301)
     True Float=1
     False Float=&FloatFromCurve(FloatCurve=DataInterface() {propertyValues {Curve {
         defaultValue=0
         keys [
          {arriveTangent=-5.7801, interpMode=RCIM_Cubic, leaveTangent=-5.7801, tangentMode=RCTM_User, time=0, value=1}
          {arriveTangent=-1.9813, interpMode=RCIM_Cubic, leaveTangent=-1.9813, tangentMode=RCTM_User, time=0.2, value=0.2}
          <at 0.5 0 RCIM_Cubic>
         ]
        }}})
     Translation Contribution:NiagaraFloat=1
     Rotation Contribution:NiagaraFloat=0
    )) {module=BlendSpace001}
   &GravityForce(Gravity=&MaskVector(
     Mask Source=0
     Mask Parameter=?
     Mask=0
     Masked Vector=<0 0 0>
     Passthrough Vector=&RandomRangeVector(
      Minimum=&VectorFromCurve(VectorCurve=DataInterface() {propertyValues {
         XCurve {keys [<at 0 0>,<at 1 0>]}
         YCurve {keys [<at 0 0>,<at 1 0>]}
         ZCurve {keys [
           {arriveTangent=1.6359, interpMode=RCIM_Cubic, leaveTangent=1.6359, tangentMode=RCTM_User, time=0, value=0}
           {arriveTangent=-28.5714, interpMode=RCIM_Cubic, leaveTangent=-28.5714, time=0.3, value=20}
           {arriveTangent=-28.5714, interpMode=RCIM_Cubic, leaveTangent=-28.5714, time=0.7, value=-20}
           {arriveTangent=1.8206, interpMode=RCIM_Cubic, leaveTangent=1.8206, tangentMode=RCTM_User, time=1, value=0}
          ]}
        }})
      Maximum=&VectorFromCurve(VectorCurve=DataInterface() {propertyValues {
         XCurve {keys [<at 0 0>,<at 1 0>]}
         YCurve {keys [<at 0 0>,<at 1 0>]}
         ZCurve {keys [
           {arriveTangent=1.6359, interpMode=RCIM_Cubic, leaveTangent=1.6359, tangentMode=RCTM_User, time=0, value=0}
           {arriveTangent=-85.7143, interpMode=RCIM_Cubic, leaveTangent=-85.7143, time=0.3, value=60}
           {arriveTangent=-85.7143, interpMode=RCIM_Cubic, leaveTangent=-85.7143, time=0.7, value=-60}
           {arriveTangent=1.8206, interpMode=RCIM_Cubic, leaveTangent=1.8206, tangentMode=RCTM_User, time=1, value=0}
          ]}
        }})
     )
    ))
   &SolveForcesAndVelocity(Clamp Velocity=false, Limit Acceleration=false)
   /SetVariables(Particles.Position=&MaskVector(
     Mask Source=0
     Mask Parameter=?
     Mask=0
     Masked Vector=&AddVectorToPosition(
      Position=$Particles.Position
      Vector=&MakeVector(
       X=0
       Y=&FloatFromCurve(
        FloatCurve=DataInterface() {propertyValues {Curve {
           defaultValue=1
           keys [
            {arriveTangent=-3.6385, interpMode=RCIM_Cubic, leaveTangent=-3.6385, tangentMode=RCTM_User, time=0, value=0}
            {tangentMode=RCTM_User, time=1, value=-1}
           ]
          }}}
        Scale Curve=0.75
       )
       Z=&FloatFromCurve(FloatCurve=DataInterface() {propertyValues {Curve {keys [
            {interpMode=RCIM_Cubic, tangentMode=RCTM_User, time=0.5, value=0}
            {arriveTangent=6.5605, interpMode=RCIM_Cubic, leaveTangent=6.5605, tangentMode=RCTM_User, time=1, value=1}
           ]}}})
      )
     )
     Passthrough Vector=&AddVectorToPosition(
      Position=$Particles.Position
      Vector=&MakeVector(
       X=0
       Y=&FloatFromCurve(
        FloatCurve=DataInterface() {propertyValues {Curve {
           defaultValue=1
           keys [
            {arriveTangent=-3.6385, interpMode=RCIM_Cubic, leaveTangent=-3.6385, tangentMode=RCTM_User, time=0, value=0}
            {tangentMode=RCTM_User, time=1, value=-1}
           ]
          }}}
        Scale Curve=0.4
       )
       Z=&FloatFromCurve(FloatCurve=DataInterface() {propertyValues {Curve {keys [
            {interpMode=RCIM_Cubic, tangentMode=RCTM_User, time=0.5, value=0}
            {arriveTangent=6.5605, interpMode=RCIM_Cubic, leaveTangent=6.5605, tangentMode=RCTM_User, time=1, value=1}
           ]}}})
      )
     )
    )) {isSetParameters=true}
   /SetVariables(Particles.Update Sparkle Mist Color:LinearColor=&MakeLinearColorFromVectorAndFloat(
     "Vector (RGB)"=&MakeVectorFromLinearColorRGB(Color=&RandomRangeLinearColor(
       Minimum=&Multiply_LinearColorByFloat(
        Linear Color=<rgba 0.476 0.881 1 0.05>
        Float=&FNTimeOfDayScalar(Night Scale=0.8, Morning Scale=0.9, Evening Scale=0.9)
       )
       Maximum=&Multiply_LinearColorByFloat(
        Linear Color=<rgba 0.4677 1 0.8984 0.05>
        Float=&FNTimeOfDayScalar(Night Scale=0.8, Morning Scale=0.9, Evening Scale=0.9)
       )
      ))
     "Float (Alpha)"=&FloatFromCurve(
      FloatCurve=DataInterface() {propertyValues {Curve {keys [
          {interpMode=RCIM_Cubic, leaveTangentWeight=0.5588, tangentMode=RCTM_User, time=0, value=0}
          {arriveTangent=7.3302, interpMode=RCIM_Cubic, leaveTangent=-2.5107, tangentMode=RCTM_Break, time=0.3, value=1}
          {arriveTangent=0.5709, interpMode=RCIM_Cubic, leaveTangent=0.5709, tangentMode=RCTM_User, time=1, value=0}
         ]}}}
      Scale Curve=&FN_FrontEndFloatModifier(Gameplay=0.07, Front End=0.04)
     )
    )) {isSetParameters=true}
   /SetVariables(Particles.Mist Droplets Color:LinearColor=&MakeLinearColorFromVectorAndFloat(
     "Vector (RGB)"=&MakeVectorFromLinearColorRGB(Color=<rgba 4.077 8.9434 10 1>)
     "Float (Alpha)"=&FloatFromCurve(FloatCurve=DataInterface() {propertyValues {Curve {keys [
          {interpMode=RCIM_Cubic, leaveTangentWeight=0.1671, tangentMode=RCTM_User, time=0, value=0}
          {interpMode=RCIM_Cubic, tangentMode=RCTM_User, time=0.5005, value=1}
          {arriveTangentWeight=0.6924, interpMode=RCIM_Cubic, tangentMode=RCTM_User, time=1, value=0}
         ]}}})
    )) {isSetParameters=true}
  ]
  rendererClasses [&NiagaraSpriteRendererProperties]
  renderers [
   {rendererClass=&NiagaraSpriteRendererProperties}
   {rendererClass=&NiagaraSpriteRendererProperties, rendererIndex=1}
   {rendererClass=&NiagaraSpriteRendererProperties, rendererIndex=2}
  ]
 }
 Mist_Condensed {
  bEnabled=false
  emitterUpdateScript [
   &EmitterState(Life Cycle Mode=::NewEnumerator1, Loop Duration=1, UseLoopDelay=false)
   &SpawnRate(SpawnRate=3, Use Spawn Probability=false) {module=SpawnRate001}
   &SpawnRate(SpawnGroup=1, SpawnRate=1.9, Use Spawn Probability=false)
   &SpawnRate(SpawnGroup=2, SpawnRate=6.5, Use Spawn Probability=false) {module=SpawnRate002}
   &SpawnRate(SpawnGroup=3, SpawnRate=0.7, Use Spawn Probability=false) {module=SpawnRate003}
   &SpawnRate(SpawnGroup=4, SpawnRate=0.7, Use Spawn Probability=false) {module=SpawnRate004}
   &SpawnRate(SpawnGroup=5, SpawnRate=1.25, Use Spawn Probability=false) {module=SpawnRate005}
   &SpawnRate(SpawnGroup=6, SpawnRate=8, Use Spawn Probability=false) {module=SpawnRate006}
   &SpawnRate(SpawnGroup=7, SpawnRate=4.3, Use Spawn Probability=false) {module=SpawnRate007}
   &SpawnRate(SpawnGroup=8, SpawnRate=0.5, Use Spawn Probability=false) {module=SpawnRate008}
  ]
  particleSpawnScript [
   /SetVariables(Particles.VisibilityTag=&SpawnGroup()) {isSetParameters=true}
   &InitializeParticle(
    Color Mode=::NewEnumerator0
    Lifetime=&MaskFloat(
     Mask Source=0
     Mask Parameter=?
     Mask=0
     Masked Float=&RandomRangeFloat(Minimum=2, Maximum=3)
     Passthrough Float=&MaskFloat(
      Mask Source=0
      Mask Parameter=?
      Mask=1
      Masked Float=&RandomRangeFloat(Minimum=0.6, Maximum=2)
      Passthrough Float=&MaskFloat(
       Mask Source=0
       Mask Parameter=?
       Mask=2
       Masked Float=&RandomRangeFloat(Minimum=1, Maximum=2.5)
       Passthrough Float=&MaskFloat(
        Mask Source=0
        Mask Parameter=?
        Mask=3
        Masked Float=&RandomRangeFloat(Minimum=3.5, Maximum=4.5)
        Passthrough Float=&MaskFloat(
         Mask Source=0
         Mask Parameter=?
         Mask=4
         Masked Float=&RandomRangeFloat(Minimum=3.5, Maximum=4.5)
         Passthrough Float=&RandomRangeFloat(Minimum=3.5, Maximum=4.5)
        )
       )
      )
     )
    )
    UsePositionOffset=false
    Write Material Random=true
   )
   /SetVariables(
    Particles.0 - Mist Color:LinearColor=&RandomRangeLinearColor(
     Minimum=&Multiply_LinearColorByFloat(
      Linear Color=<rgba 0.7353 0.9636 1 0.02>
      Float=&FNTimeOfDayScalar(Night Scale=0.5, Morning Scale=0.65, Evening Scale=0.65)
     )
     Maximum=&Multiply_LinearColorByFloat(
      Linear Color=<rgba 0.4498 0.8309 1 0.02>
      Float=&FNTimeOfDayScalar(Night Scale=0.5, Morning Scale=0.65, Evening Scale=0.65)
     )
    )
    Particles.0 - Mist Sprite Size:Vector2f=&Vector2DFromFloat(Value=&RandomRangeFloat(Minimum=70, Maximum=140))
   ) {isSetParameters=true}
   /SetVariables(
    Particles.1 - Sparkle Mist Color:LinearColor=&RandomRangeLinearColor(
     Minimum=&Multiply_LinearColorByFloat(
      Linear Color=<rgba 0.476 0.881 1 0.05>
      Float=&FNTimeOfDayScalar(Night Scale=0.8, Morning Scale=0.9, Evening Scale=0.9)
     )
     Maximum=&Multiply_LinearColorByFloat(
      Linear Color=<rgba 0.4677 1 0.8984 0.05>
      Float=&FNTimeOfDayScalar(Night Scale=0.8, Morning Scale=0.9, Evening Scale=0.9)
     )
    )
    Particles.1 - Sparkle Mist Size:Vector2f=&Vector2DFromFloat(Value=&RandomRangeFloat(Minimum=70, Maximum=140))
   ) {isSetParameters=true}
   /SetVariables(
    Particles.2 - Droplets Color:LinearColor=<rgba 4.077 8.9434 10 1>
    Particles.2 - Droplets Size:Vector2f=&Vector2DFromFloat(Value=&RandomRangeFloat(Minimum=0.5, Maximum=2))
   ) {isSetParameters=true}
   /SetVariables(
    Particles.Ground Mist Color:LinearColor=&Multiply_LinearColorByFloat(
     Linear Color=&RandomRangeLinearColor(Minimum=<rgba 0.7353 0.9636 1 0.02>, Maximum=<rgba 0.4498 0.8309 1 0.02>)
     Float=&FNTimeOfDayScalar(Night Scale=0.62, Morning Scale=0.8, Evening Scale=0.8)
    )
    Particles.Ground Mist Size:Vector2f=&Vector2DFromFloat(Value=&RandomRangeFloat(Minimum=90, Maximum=120))
   ) {isSetParameters=true}
   /SetVariables() {isSetParameters=true}
   /SetVariables() {isSetParameters=true}
   /SetVariables() {isSetParameters=true}
   &ShapeLocation(
    MaskSpawnGroup=false
    Non Uniform Scale=&MaskVector(
     Mask Source=0
     Mask Parameter=?
     Mask=0
     Masked Vector=<0 0 0>
     Passthrough Vector=&MaskVector(
      Mask Source=0
      Mask Parameter=?
      Mask=1
      Masked Vector=<0 0 0>
      Passthrough Vector=&MaskVector(
       Mask Source=0
       Mask Parameter=?
       Mask=2
       Masked Vector=<0 0 0>
       Passthrough Vector=&MaskVector(
        Mask Source=0
        Mask Parameter=?
        Mask=3
        Masked Vector=<0 0 0>
        Passthrough Vector=&MaskVector(
         Mask Source=0
         Mask Parameter=?
         Mask=4
         Masked Vector=<0 0 0>
         Passthrough Vector=&MaskVector(
          Mask Source=0
          Mask Parameter=?
          Mask=5
          Masked Vector=<0 0 0>
          Passthrough Vector=&MaskVector(
           Mask Source=0
           Mask Parameter=?
           Mask=6
           Masked Vector=<0 0 0>
           Passthrough Vector=&MaskVector(Mask Source=0, Mask Parameter=?, Mask=7, Masked Vector=<0 0 0>, Passthrough Vector=<1 1 1>)
          )
         )
        )
       )
      )
     )
    )
    Override Seed=false
    Sphere Radius=&MaskFloat(
     Mask Source=0
     Mask Parameter=?
     Mask=0
     Masked Float=0
     Passthrough Float=&MaskFloat(
      Mask Source=0
      Mask Parameter=?
      Mask=1
      Masked Float=0
      Passthrough Float=&MaskFloat(
       Mask Source=0
       Mask Parameter=?
       Mask=2
       Masked Float=0
       Passthrough Float=&MaskFloat(
        Mask Source=0
        Mask Parameter=?
        Mask=3
        Masked Float=0
        Passthrough Float=&MaskFloat(
         Mask Source=0
         Mask Parameter=?
         Mask=4
         Masked Float=0
         Passthrough Float=&MaskFloat(
          Mask Source=0
          Mask Parameter=?
          Mask=5
          Masked Float=0
          Passthrough Float=&MaskFloat(
           Mask Source=0
           Mask Parameter=?
           Mask=6
           Masked Float=0
           Passthrough Float=&MaskFloat(Mask Source=0, Mask Parameter=?, Mask=7, Masked Float=0, Passthrough Float=1)
          )
         )
        )
       )
      )
     )
    )
    UseOwnerScale=true
   )
   &ShapeLocation(
    Box Size=&MaskVector(
     Mask Source=0
     Mask Parameter=?
     Mask=0
     Masked Vector=<0 0 0>
     Passthrough Vector=&MaskVector(
      Mask Source=0
      Mask Parameter=?
      Mask=1
      Masked Vector=<0 0 0>
      Passthrough Vector=&MaskVector(
       Mask Source=0
       Mask Parameter=?
       Mask=2
       Masked Vector=<0 0 0>
       Passthrough Vector=&MaskVector(
        Mask Source=0
        Mask Parameter=?
        Mask=3
        Masked Vector=<0 0 0>
        Passthrough Vector=&MaskVector(
         Mask Source=0
         Mask Parameter=?
         Mask=4
         Masked Vector=<0 0 0>
         Passthrough Vector=&MaskVector(
          Mask Source=0
          Mask Parameter=?
          Mask=5
          Masked Vector=<0 0 0>
          Passthrough Vector=&MaskVector(
           Mask Source=0
           Mask Parameter=?
           Mask=6
           Masked Vector=<0 0 0>
           Passthrough Vector=&MaskVector(Mask Source=0, Mask Parameter=?, Mask=7, Masked Vector=<0 0 0>, Passthrough Vector=<0 0 0>)
          )
         )
        )
       )
      )
     )
    )
    MaskSpawnGroup=false
    Non Uniform Scale=&MaskVector(
     Mask Source=0
     Mask Parameter=?
     Mask=0
     Masked Vector=<0 0 0>
     Passthrough Vector=&MaskVector(
      Mask Source=0
      Mask Parameter=?
      Mask=1
      Masked Vector=<0 0 0>
      Passthrough Vector=&MaskVector(
       Mask Source=0
       Mask Parameter=?
       Mask=2
       Masked Vector=<0 0 0>
       Passthrough Vector=&MaskVector(
        Mask Source=0
        Mask Parameter=?
        Mask=3
        Masked Vector=<0 0 0>
        Passthrough Vector=&MaskVector(
         Mask Source=0
         Mask Parameter=?
         Mask=4
         Masked Vector=<0 0 0>
         Passthrough Vector=&MaskVector(
          Mask Source=0
          Mask Parameter=?
          Mask=5
          Masked Vector=<0 0 0>
          Passthrough Vector=&MaskVector(
           Mask Source=0
           Mask Parameter=?
           Mask=6
           Masked Vector=<0 0 0>
           Passthrough Vector=&MaskVector(Mask Source=0, Mask Parameter=?, Mask=7, Masked Vector=<0 0 0>, Passthrough Vector=<1 1 1>)
          )
         )
        )
       )
      )
     )
    )
    Override Seed=false
    Shape Primitive=::NewEnumerator2
    Surface Only=false
    UseOwnerScale=true
   ) {module=ShapeLocation001}
   &CameraOffset(Camera Offset Amount=&MaskFloat(
     Mask Source=0
     Mask Parameter=?
     Mask=0
     Masked Float=0
     Passthrough Float=&MaskFloat(
      Mask Source=0
      Mask Parameter=?
      Mask=1
      Masked Float=0
      Passthrough Float=&MaskFloat(
       Mask Source=0
       Mask Parameter=?
       Mask=2
       Masked Float=0
       Passthrough Float=&MaskFloat(
        Mask Source=0
        Mask Parameter=?
        Mask=3
        Masked Float=0
        Passthrough Float=&MaskFloat(
         Mask Source=0
         Mask Parameter=?
         Mask=4
         Masked Float=0
         Passthrough Float=&MaskFloat(
          Mask Source=0
          Mask Parameter=?
          Mask=5
          Masked Float=0
          Passthrough Float=&MaskFloat(
           Mask Source=0
           Mask Parameter=?
           Mask=6
           Masked Float=0
           Passthrough Float=&MaskFloat(Mask Source=0, Mask Parameter=?, Mask=7, Masked Float=0, Passthrough Float=0)
          )
         )
        )
       )
      )
     )
    ))
  ]
  particleUpdateScript [
   &ParticleState()
   /SetVariables(
    Particles.0 - Update Mist Color:LinearColor=&MakeLinearColorFromVectorAndFloat(
     "Vector (RGB)"=&MakeVectorFromLinearColorRGB(Color=Linked() {linkedVariable {name=Particles.Initial.0 - Mist Color, type=LinearColor}})
     "Float (Alpha)"=&FloatFromCurve(FloatCurve=DataInterface() {propertyValues {Curve {keys [
          <at 0 0 RCIM_Cubic>
          {arriveTangent=6.0452, interpMode=RCIM_Cubic, leaveTangent=6.0452, tangentMode=RCTM_User, time=0.4, value=1}
          {arriveTangent=-5.3053, interpMode=RCIM_Cubic, leaveTangent=-5.3053, tangentMode=RCTM_User, time=0.5, value=1}
          <at 1 0 RCIM_Cubic>
         ]}}})
    )
    Particles.0 - Update Mist Size:Vector2f=&Multiply_Vector2D_ByFloat(
     Float=&FloatFromCurve(FloatCurve=DataInterface() {propertyValues {Curve {keys [
          {arriveTangent=3.3351, interpMode=RCIM_Cubic, leaveTangent=3.3351, tangentMode=RCTM_User, time=0, value=0}
          {arriveTangent=0.0592, interpMode=RCIM_Cubic, leaveTangent=0.0592, tangentMode=RCTM_User, time=0.7, value=0.65}
          {arriveTangent=2.1747, interpMode=RCIM_Cubic, leaveTangent=2.1747, tangentMode=RCTM_User, time=1, value=1}
         ]}}})
     Vector2D=Linked() {linkedVariable {name=Particles.0 - Mist Sprite Size, type=Vector2f}}
    )
   ) {isSetParameters=true}
   /SetVariables(
    Particles.1 - Update Sparkle Mist COlor:LinearColor=&MakeLinearColorFromVectorAndFloat(
     "Vector (RGB)"=&MakeVectorFromLinearColorRGB(Color=Linked() {linkedVariable {name=Particles.Initial.1 - Sparkle Mist Color, type=LinearColor}})
     "Float (Alpha)"=&FloatFromCurve(FloatCurve=DataInterface() {propertyValues {Curve {keys [
          {interpMode=RCIM_Cubic, leaveTangentWeight=0.5588, tangentMode=RCTM_User, time=0, value=0}
          {arriveTangent=7.3302, interpMode=RCIM_Cubic, leaveTangent=-2.5107, tangentMode=RCTM_Break, time=0.3, value=1}
          {arriveTangent=0.5709, interpMode=RCIM_Cubic, leaveTangent=0.5709, tangentMode=RCTM_User, time=1, value=0}
         ]}}})
    )
    Particles.1 - Update Sparkle Mist Size:Vector2f=&Multiply_Vector2D_ByFloat(
     Float=&FloatFromCurve(FloatCurve=DataInterface() {propertyValues {Curve {keys [
          {arriveTangent=3.3351, interpMode=RCIM_Cubic, leaveTangent=3.3351, tangentMode=RCTM_User, time=0, value=0}
          {arriveTangent=0.0592, interpMode=RCIM_Cubic, leaveTangent=0.0592, tangentMode=RCTM_User, time=0.7, value=0.65}
          {arriveTangent=2.1747, interpMode=RCIM_Cubic, leaveTangent=2.1747, tangentMode=RCTM_User, time=1, value=1}
         ]}}})
     Vector2D=Linked() {linkedVariable {name=Particles.1 - Sparkle Mist Size, type=Vector2f}}
    )
   ) {isSetParameters=true}
   /SetVariables() {isSetParameters=true}
   /SetVariables() {isSetParameters=true}
   &GravityForce(Gravity=&MaskVector(
     Mask Source=0
     Mask Parameter=?
     Mask=0
     Masked Vector=<0 0 0>
     Passthrough Vector=&MaskVector(
      Mask Source=0
      Mask Parameter=?
      Mask=1
      Masked Vector=<0 0 0>
      Passthrough Vector=&MaskVector(
       Mask Source=0
       Mask Parameter=?
       Mask=2
       Masked Vector=<0 0 0>
       Passthrough Vector=&MaskVector(
        Mask Source=0
        Mask Parameter=?
        Mask=3
        Masked Vector=<0 0 0>
        Passthrough Vector=&MaskVector(
         Mask Source=0
         Mask Parameter=?
         Mask=4
         Masked Vector=<0 0 0>
         Passthrough Vector=&MaskVector(
          Mask Source=0
          Mask Parameter=?
          Mask=5
          Masked Vector=<0 0 0>
          Passthrough Vector=&MaskVector(
           Mask Source=0
           Mask Parameter=?
           Mask=6
           Masked Vector=<0 0 0>
           Passthrough Vector=&MaskVector(Mask Source=0, Mask Parameter=?, Mask=7, Masked Vector=<0 0 0>, Passthrough Vector=<0 0 0>)
          )
         )
        )
       )
      )
     )
    ))
   &SolveForcesAndVelocity(Clamp Velocity=false, Limit Acceleration=false)
  ]
  rendererClasses [&NiagaraSpriteRendererProperties]
  renderers [
   {rendererClass=&NiagaraSpriteRendererProperties}
   {rendererClass=&NiagaraSpriteRendererProperties, rendererIndex=1}
   {rendererClass=&NiagaraSpriteRendererProperties, rendererIndex=2}
   {rendererClass=&NiagaraSpriteRendererProperties, rendererIndex=3}
   {rendererClass=&NiagaraSpriteRendererProperties, rendererIndex=4}
   {rendererClass=&NiagaraSpriteRendererProperties, rendererIndex=5}
   {rendererClass=&NiagaraSpriteRendererProperties, rendererIndex=6}
   {rendererClass=&NiagaraSpriteRendererProperties, rendererIndex=7}
   {rendererClass=&NiagaraSpriteRendererProperties, rendererIndex=8}
  ]
 }
 New_Mist {
  emitterUpdateScript [
   &EmitterState()
   &SpawnRate(SpawnRate=0.7, Use Spawn Probability=false)
   &SpawnRate(SpawnGroup=1, SpawnRate=0.7, Use Spawn Probability=false) {module=SpawnRate001}
   &SpawnRate(SpawnGroup=2, SpawnRate=1.25, Use Spawn Probability=false) {module=SpawnRate002}
  ]
  particleSpawnScript [
   &InitializeParticle(
    Color=&Multiply_LinearColorByFloat(
     Linear Color=&RandomRangeLinearColor(Minimum=<rgba 0.7353 0.9636 1 0.02>, Maximum=<rgba 0.4498 0.8309 1 0.02>)
     Float=&FNTimeOfDayScalar(Night Scale=0.62, Morning Scale=0.8, Evening Scale=0.8)
    )
    Lifetime Max=4.5
    Lifetime Min=3.5
    Lifetime Mode=::NewEnumerator1
    Sprite Size Mode=::NewEnumerator3
    Uniform Sprite Size=100
    UsePositionOffset=false
    Write Material Random=true
   )
   &ShapeLocation(
    MaskSpawnGroup=false
    Non Uniform Scale=&MaskVector(
     Mask Source=0
     Mask Parameter=?
     Mask=0
     Masked Vector=<1.2 1.5 1>
     Passthrough Vector=&MaskVector(Mask Source=0, Mask Parameter=?, Mask=1, Masked Vector=<1.2 1.5 1>, Passthrough Vector=<3.6 1 1>)
    )
    Offset=&MaskVector(
     Mask Source=0
     Mask Parameter=?
     Mask=0
     Masked Vector=<-50 60 20>
     Passthrough Vector=&MaskVector(Mask Source=0, Mask Parameter=?, Mask=1, Masked Vector=<50 60 20>, Passthrough Vector=<0 110 30>)
    )
    Offset Coordinate Space=::Local
    Offset Mode=::NewEnumerator0
    Override Seed=false
    Sphere Radius=20
    UseOwnerScale=true
   )
  ]
  particleUpdateScript [
   &ParticleState()
   &ScaleSpriteSize(Uniform Curve Sprite Scale=DataInterface() {propertyValues {Curve {keys [
        {arriveTangent=0.6277, interpMode=RCIM_Cubic, leaveTangent=0.6277, tangentMode=RCTM_User, time=0, value=0}
        {arriveTangent=5.5396, interpMode=RCIM_Cubic, leaveTangent=5.5396, tangentMode=RCTM_User, time=0.2122, value=0.613}
        {arriveTangent=4.2767, interpMode=RCIM_Cubic, leaveTangent=4.2767, tangentMode=RCTM_Break, time=0.5022, value=1.6221}
        {interpMode=RCIM_Cubic, leaveTangent=6.4626, time=0.505, value=0}
       ]}}})
   /SetVariables(Particles.Position=&AddVectorToPosition(
     Position=$Particles.Position
     Vector=&MakeVector(
      X=0
      Y=&FloatFromCurve(
       FloatCurve=DataInterface() {propertyValues {Curve {
          defaultValue=1
          keys [
           {arriveTangent=-5.6911, interpMode=RCIM_Cubic, leaveTangent=-5.6911, tangentMode=RCTM_User, time=0, value=0}
           {arriveTangent=1.089, interpMode=RCIM_Cubic, leaveTangent=1.089, tangentMode=RCTM_User, time=0.8729, value=-1.3198}
           {arriveTangent=27.7247, tangentMode=RCTM_User, time=1.0021, value=1.9595}
          ]
         }}}
       Scale Curve=&MakeCustomFloatFromBool(
        Bool=&SetBoolByFloatComparison(A=&VectorLength(Vector=$Engine.Owner.Velocity), B=50)
        True Float=1
        False Float=1
       )
      )
      Z=&FloatFromCurve(
       FloatCurve=DataInterface() {propertyValues {Curve {keys [
           {interpMode=RCIM_Cubic, tangentMode=RCTM_User, time=0.5, value=0}
           {arriveTangent=3.8789, interpMode=RCIM_Cubic, leaveTangent=3.8789, tangentMode=RCTM_User, time=0.816, value=0.1762}
           {arriveTangent=39.4449, interpMode=RCIM_Cubic, leaveTangent=39.4449, tangentMode=RCTM_User, time=0.9437, value=3.6812}
           <at 0.9991 2.1624 RCIM_Cubic>
          ]}}}
       Scale Curve=&MakeCustomFloatFromBool(
        Bool=&SetBoolByFloatComparison(A=&VectorLength(Vector=$Engine.Owner.Velocity), B=50)
        True Float=0.5
        False Float=1
       )
      )
     )
    )) {isSetParameters=true}
   &GravityForce(Gravity=&VectorFromCurve(VectorCurve=DataInterface() {propertyValues {
       XCurve {keys [<at 0 0>,<at 1 0>]}
       YCurve {keys [
         {interpMode=RCIM_Cubic, tangentMode=RCTM_User, time=0, value=0}
         {arriveTangent=-27.5228, interpMode=RCIM_Cubic, leaveTangent=-27.5228, tangentMode=RCTM_User, time=1, value=-10}
        ]}
       ZCurve {keys [
         {arriveTangent=13.4933, interpMode=RCIM_Cubic, leaveTangent=13.4933, tangentMode=RCTM_User, time=0, value=0}
         {interpMode=RCIM_Cubic, tangentMode=RCTM_User, time=1, value=5}
        ]}
      }}))
   &SolveForcesAndVelocity(Clamp Velocity=false, Limit Acceleration=false)
   &ScaleSpriteSizeBySpeed(Max Scale Factor=<2 1>, Velocity Threshold=10)
   &BlendSpace("Position Blend (World <-> Local)":NiagaraFloat=&MakeCustomFloatFromBool(
     Bool=&SetBoolByFloatComparison(A=&VectorLength(Vector=$Engine.Owner.Velocity), B=301)
     True Float=1
     False Float=&FloatFromCurve(FloatCurve=DataInterface() {propertyValues {Curve {
         defaultValue=0
         keys [
          {arriveTangent=-5.7801, interpMode=RCIM_Cubic, leaveTangent=-5.7801, tangentMode=RCTM_User, time=0, value=1}
          {arriveTangent=-1.9813, interpMode=RCIM_Cubic, leaveTangent=-1.9813, tangentMode=RCTM_User, time=0.2, value=0.2}
          <at 0.5 0 RCIM_Cubic>
         ]
        }}})
     Translation Contribution:NiagaraFloat=1
     Rotation Contribution:NiagaraFloat=0
    )) {module=BlendSpace001}
   &DynamicMaterialParameters(Module.Index0Float0Write=true, Module.Index0Float1Write=true, Module.Index0Float2Write=true, Module.Index0Float3Write=true)
   &SubUVAnimation(
    "Sprite Renderer "=DataInterface() {propertyValues {SpriteRenderer=&self:New_Mist_0.NiagaraSpriteRendererProperties_0}}
    SubUV Lookup Index Scale:NiagaraFloat=&FloatFromCurve(FloatCurve=DataInterface() {propertyValues {Curve {keys [
         {arriveTangent=0.8228, interpMode=RCIM_Cubic, leaveTangent=0.8228, tangentMode=RCTM_User, time=-0.0072, value=0}
         {arriveTangent=1.6719, interpMode=RCIM_Cubic, leaveTangent=1.6719, tangentMode=RCTM_User, time=1, value=1}
        ]}}})
    UseEndFrame=false
    UseStartFrame=false
   )
   &ScaleColor(
    Scale Alpha=&FloatFromCurve(
     FloatCurve=DataInterface() {propertyValues {Curve {keys [
         {interpMode=RCIM_Cubic, leaveTangentWeight=0.1209, tangentMode=RCTM_Break, time=0.012, value=0.0601}
         {arriveTangent=5.6843, interpMode=RCIM_Cubic, tangentMode=RCTM_Break, time=0.3318, value=0.8948}
         {arriveTangentWeight=0.2412, interpMode=RCIM_Cubic, tangentMode=RCTM_Break, time=0.5, value=0}
        ]}}}
     Scale Curve=&FN_FrontEndFloatModifier(Gameplay=1, Front End=0.6)
    )
    ScaleA=true
    ScaleRGB=true
   )
  ]
  rendererClasses [&NiagaraSpriteRendererProperties]
  renderers [{rendererClass=&NiagaraSpriteRendererProperties}]
 }
}
summary {
 systemName=NS_AquaPeony_Loop
 emitters [
  {emitterName=Mist002, rendererClasses [&NiagaraSpriteRendererProperties]}
  {emitterName=New_Mist, rendererClasses [&NiagaraSpriteRendererProperties]}
  {emitterName=Light_Glints001, rendererClasses [&NiagaraSpriteRendererProperties]}
  {bEnabled=false, emitterName=Mist_Condensed, rendererClasses [&NiagaraSpriteRendererProperties]}
  {emitterName=Glow_Fake_Light, rendererClasses []}
 ]
 userVariables [
  User.bIsFrontEnd {type=NiagaraBool}
  User.bIsFrontEndPreview {type=NiagaraBool}
  User.Intensity {type=NiagaraFloat, defaultValue=0.68, description=""}
  User.OwnerLocalVelocity {type=Vector3f, description="", defaultValue {struct=Vector3f, value=<0 0 0>}}
 ]
}
