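# Niagara system for the AquaPeony player emote, producing water splashes and sand at the feet plus droplets from the hair.
# Timed bursts alternate between feet via spawn groups and skeletal mesh sampling, while hair droplets use per-group masks for offset and motion.
# Owner speed switches splash timing, size, gravity and world-to-local position blending.
# Front-end modifiers, time of day, and platform or quality checks adjust placement and color.
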NiagaraSystem /BRCosmetics/Animation/Game/MainPlayer/Emotes/AquaPeony/FX/NS_AquaPeony_Splashes
dependencies {
 usedDataInterfaces [
  &NiagaraDataInterfaceVectorCurve
  &NiagaraDataInterfaceCurve
  &NiagaraDataInterfacePlatformSet
  &NiagaraDataInterfaceArrayColor
  &NiagaraDataInterfaceSpriteRendererInfo
  &NiagaraDataInterfaceSkeletalMesh
 ]
 usedDynamicInputs []
 usedModules [
  &SystemState
  &BlendSpaceHelper
  &self:SystemSpawnScript.SystemScriptSource.SystemScriptGraph.NiagaraNodeAssignment_3.SetVariables_#1
  &self:SystemSpawnScript.SystemScriptSource.SystemScriptGraph.NiagaraNodeAssignment_4.SetVariables_#2
  &self:SystemSpawnScript.SystemScriptSource.SystemScriptGraph.NiagaraNodeAssignment_5.SetVariables_#3
 ]
}
emitters {
 New_HairDroplets {
  emitterUpdateScript [
   &EmitterState()
   &SpawnBurst_Instantaneous(
    Spawn Count=&RandomRangeInt(Minimum=2, Maximum=3)
    Spawn Group=1
    Spawn Time=10.5
    Use Spawn Probability=false
    UseLoopCountLimit=false
   )
   &SpawnBurst_Instantaneous(
    Spawn Count=&RandomRangeInt(Minimum=2, Maximum=3)
    Spawn Time=11.47
    Use Spawn Probability=false
    UseLoopCountLimit=false
   ) {module=SpawnBurst_Instantaneous001}
   &SpawnBurst_Instantaneous(
    Spawn Count=&RandomRangeInt(Minimum=1, Maximum=2)
    Spawn Group=3
    Spawn Time=10.5
    Use Spawn Probability=false
    UseLoopCountLimit=false
   ) {module=SpawnBurst_Instantaneous002}
   &SpawnBurst_Instantaneous(
    Spawn Count=&RandomRangeInt(Minimum=1, Maximum=2)
    Spawn Group=2
    Spawn Time=11.47
    Use Spawn Probability=false
    UseLoopCountLimit=false
   ) {module=SpawnBurst_Instantaneous003}
  ]
  particleSpawnScript [
   &InitializeParticle(
    Color Maximum=&Multiply_LinearColorByFloat(
     Linear Color=<rgba 0.4622 0.7474 1 0.8>
     Float=&FNTimeOfDayScalar(Night Scale=0.7, Morning Scale=0.8, Evening Scale=0.8)
    )
    Color Minimum=&Multiply_LinearColorByFloat(
     Linear Color=<rgba 0.2497 0.7526 1 0.5>
     Float=&FNTimeOfDayScalar(Night Scale=0.7, Morning Scale=0.8, Evening Scale=0.8)
    )
    Color Mode=::NewEnumerator2
    Lifetime Max=&MaskFloat(
     Mask Source=0
     Mask Parameter=?
     Mask=0
     Masked Float=1
     Passthrough Float=&MaskFloat(
      Mask Source=0
      Mask Parameter=?
      Mask=1
      Masked Float=1
      Passthrough Float=&MaskFloat(Mask Source=0, Mask Parameter=?, Mask=2, Masked Float=1.5, Passthrough Float=1.5)
     )
    )
    Lifetime Min:NiagaraFloat=&MaskFloat(
     Mask Source=0
     Mask Parameter=?
     Mask=0
     Masked Float=0.5
     Passthrough Float=&MaskFloat(
      Mask Source=0
      Mask Parameter=?
      Mask=1
      Masked Float=0.5
      Passthrough Float=&MaskFloat(Mask Source=0, Mask Parameter=?, Mask=2, Masked Float=0.7, Passthrough Float=0.7)
     )
    )
    Lifetime Mode=::NewEnumerator1
    Sprite Size Mode=::NewEnumerator1
    Uniform Sprite Size Max=0.6
    Uniform Sprite Size Min=0.4
    UsePositionOffset=false
    Write Material Random=true
   )
   &SkeletalMeshLocation(
    Bone Sampling Mode=::NewEnumerator4
    MaskSpawnGroup=false
    Sampled Position Offset=&MaskVector(
     Mask Source=0
     Mask Parameter=?
     Mask=0
     Masked Vector=<6.5 -1 10>
     Passthrough Vector=&MaskVector(
      Mask Source=0
      Mask Parameter=?
      Mask=1
      Masked Vector=<-8.5 -1 10>
      Passthrough Vector=&MaskVector(Mask Source=0, Mask Parameter=?, Mask=2, Masked Vector=<6.5 -1 12>, Passthrough Vector=<-8.5 -1 12>)
     )
    )
    Sampled Position Offset Coordinate Space:ENiagaraSkelMeshTransforms=::NewEnumerator2
    Skeletal Mesh=$System.Hair Droplets Spawn
   )
   &ShapeLocation(
    MaskSpawnGroup=false
    Non Uniform Scale=&MaskVector(
     Mask Source=0
     Mask Parameter=?
     Mask=0
     Masked Vector=<0.5 1.8 2>
     Passthrough Vector=&MaskVector(
      Mask Source=0
      Mask Parameter=?
      Mask=1
      Masked Vector=<0.5 1.8 3>
      Passthrough Vector=&MaskVector(Mask Source=0, Mask Parameter=?, Mask=2, Masked Vector=<0.5 1.8 2>, Passthrough Vector=<0.5 1.8 3>)
     )
    )
    Override Seed=false
    Sphere Radius=3
    UseOwnerScale=true
   )
   &AddVelocity(Velocity=&MaskVector(
     Mask Source=0
     Mask Parameter=?
     Mask=0
     Masked Vector=&VectorFromCurve(VectorCurve=DataInterface() {propertyValues {
        XCurve {keys [{arriveTangent=20.7603, interpMode=RCIM_Cubic, leaveTangent=20.7603, tangentMode=RCTM_User, time=0, value=4}, <at 0.5 7>]}
        YCurve {keys [{arriveTangent=-33.0457, interpMode=RCIM_Cubic, leaveTangent=-33.0457, tangentMode=RCTM_User, time=0, value=0}, <at 1 -10 RCIM_Cubic>]}
        ZCurve {keys [
          {arriveTangent=-129.6528, interpMode=RCIM_Cubic, leaveTangent=-129.6528, tangentMode=RCTM_User, time=0, value=-20}
          {arriveTangent=6.3156, interpMode=RCIM_Cubic, leaveTangent=6.3156, tangentMode=RCTM_User, time=0.5, value=-40}
         ]}
       }})
     Passthrough Vector=&MaskVector(
      Mask Source=0
      Mask Parameter=?
      Mask=1
      Masked Vector=&VectorFromCurve(VectorCurve=DataInterface() {propertyValues {
         XCurve {keys [{arriveTangent=-57.7732, interpMode=RCIM_Cubic, leaveTangent=-57.7732, tangentMode=RCTM_User, time=0, value=-15}, <at 1 -30>]}
         YCurve {keys [{arriveTangent=-15.2732, interpMode=RCIM_Cubic, leaveTangent=-15.2732, tangentMode=RCTM_User, time=0, value=0}, <at 1 -7.5 RCIM_Cubic>]}
         ZCurve {keys [
           {arriveTangent=-92.5548, interpMode=RCIM_Cubic, leaveTangent=-92.5548, tangentMode=RCTM_User, time=0.1, value=0}
           {arriveTangent=6.3156, interpMode=RCIM_Cubic, leaveTangent=6.3156, tangentMode=RCTM_User, time=1, value=-30}
          ]}
        }})
      Passthrough Vector=&MaskVector(
       Mask Source=0
       Mask Parameter=?
       Mask=2
       Masked Vector=&VectorFromCurve(VectorCurve=DataInterface() {propertyValues {
          XCurve {keys [{arriveTangent=20.7603, interpMode=RCIM_Cubic, leaveTangent=20.7603, tangentMode=RCTM_User, time=0, value=4}, <at 0.5 7>]}
          YCurve {keys [{arriveTangent=-33.0457, interpMode=RCIM_Cubic, leaveTangent=-33.0457, tangentMode=RCTM_User, time=0, value=0}, <at 1 -10 RCIM_Cubic>]}
          ZCurve {keys [
            {arriveTangent=-129.6528, interpMode=RCIM_Cubic, leaveTangent=-129.6528, tangentMode=RCTM_User, time=0, value=-20}
            {arriveTangent=6.3156, interpMode=RCIM_Cubic, leaveTangent=6.3156, tangentMode=RCTM_User, time=0.5, value=-40}
           ]}
         }})
       Passthrough Vector=&VectorFromCurve(VectorCurve=DataInterface() {propertyValues {
          XCurve {keys [{arriveTangent=-57.7732, interpMode=RCIM_Cubic, leaveTangent=-57.7732, tangentMode=RCTM_User, time=0, value=-15}, <at 1 -30>]}
          YCurve {keys [{arriveTangent=-15.2732, interpMode=RCIM_Cubic, leaveTangent=-15.2732, tangentMode=RCTM_User, time=0, value=0}, <at 1 -7.5 RCIM_Cubic>]}
          ZCurve {keys [
            {arriveTangent=-92.5548, interpMode=RCIM_Cubic, leaveTangent=-92.5548, tangentMode=RCTM_User, time=0.1, value=0}
            {arriveTangent=6.3156, interpMode=RCIM_Cubic, leaveTangent=6.3156, tangentMode=RCTM_User, time=1, value=-30}
           ]}
         }})
      )
     )
    ))
   &AddVelocity(
    Module.UseConeFalloff=false
    Module.UseVelDistribution=false
    Override Seed=false
    Velocity Mode=::NewEnumerator2
    Velocity Speed=&MaskFloat(
     Mask Source=0
     Mask Parameter=?
     Mask=0
     Masked Float=20
     Passthrough Float=&MaskFloat(
      Mask Source=0
      Mask Parameter=?
      Mask=1
      Masked Float=-40
      Passthrough Float=&MaskFloat(Mask Source=0, Mask Parameter=?, Mask=2, Masked Float=20, Passthrough Float=-40)
     )
    )
   ) {module=AddVelocity001}
   &CurlNoiseForce(Bias Noise=false, Noise Frequency=150, Noise Strength=5, Pan Noise=false)
   &ApplyInitialForces()
   &CameraOffset(Camera Offset Amount=&RandomRangeFloat(Minimum=-5, Maximum=10))
  ]
  particleUpdateScript [
   &ParticleState()
   &ScaleSpriteSize(Uniform Curve Sprite Scale=DataInterface() {propertyValues {Curve {
       defaultValue=0
       keys [<at 0 1 RCIM_Cubic>, {arriveTangent=-3.8025, interpMode=RCIM_Cubic, leaveTangent=-3.8025, tangentMode=RCTM_User, time=1, value=0}]
      }}})
   &BlendSpace("Position Blend (World <-> Local)":NiagaraFloat=&MakeCustomFloatFromBool(
     Bool=&SetBoolByFloatComparison(A=&VectorLength(Vector=$Engine.Owner.Velocity), B=301)
     True Float=1
     False Float=&FloatFromCurve(FloatCurve=DataInterface() {propertyValues {Curve {
         defaultValue=0
         keys [
          {arriveTangent=-0.1834, interpMode=RCIM_Cubic, leaveTangent=-0.1834, tangentMode=RCTM_User, time=0, value=1}
          {interpMode=RCIM_Cubic, tangentMode=RCTM_User, time=0.8, value=0.5}
         ]
        }}})
     Translation Contribution:NiagaraFloat=1
     Rotation Contribution:NiagaraFloat=0
    )) {module=BlendSpace001}
   &GravityForce(Gravity=&MaskVector(
     Mask Source=0
     Mask Parameter=?
     Mask=0
     Masked Vector=&VectorFromCurve(VectorCurve=DataInterface() {propertyValues {
        XCurve {keys [<at 0 0>,<at 1 0>]}
        YCurve {keys [<at 0 0>,<at 1 0>]}
        ZCurve {keys [
          {arriveTangent=5395.769, interpMode=RCIM_Cubic, leaveTangent=5395.769, tangentMode=RCTM_User, time=0, value=0}
          {arriveTangent=-51.5965, interpMode=RCIM_Cubic, leaveTangent=-51.5965, tangentMode=RCTM_User, time=0.1, value=0}
          {arriveTangent=-1533.0747, interpMode=RCIM_Cubic, leaveTangent=-1533.0747, tangentMode=RCTM_User, time=1, value=-700}
         ]}
       }})
     Passthrough Vector=&MaskVector(
      Mask Source=0
      Mask Parameter=?
      Mask=1
      Masked Vector=&VectorFromCurve(VectorCurve=DataInterface() {propertyValues {
         XCurve {keys [<at 0 0>,<at 1 0>]}
         YCurve {keys [<at 0 0>,<at 1 0>]}
         ZCurve {keys [
           {arriveTangent=-342.6536, interpMode=RCIM_Cubic, leaveTangent=-342.6536, tangentMode=RCTM_User, time=0, value=0}
           {arriveTangent=-1163.7987, interpMode=RCIM_Cubic, leaveTangent=-1163.7987, tangentMode=RCTM_User, time=1, value=-700}
          ]}
        }})
      Passthrough Vector=&MaskVector(
       Mask Source=0
       Mask Parameter=?
       Mask=2
       Masked Vector=&VectorFromCurve(VectorCurve=DataInterface() {propertyValues {
          XCurve {keys [<at 0 0>,<at 1 0>]}
          YCurve {keys [<at 0 0>,<at 1 0>]}
          ZCurve {keys [
            {arriveTangent=5395.769, interpMode=RCIM_Cubic, leaveTangent=5395.769, tangentMode=RCTM_User, time=0, value=0}
            {arriveTangent=-51.5965, interpMode=RCIM_Cubic, leaveTangent=-51.5965, tangentMode=RCTM_User, time=0.1, value=0}
            {arriveTangent=-1533.0747, interpMode=RCIM_Cubic, leaveTangent=-1533.0747, tangentMode=RCTM_User, time=1, value=-700}
           ]}
         }})
       Passthrough Vector=&VectorFromCurve(VectorCurve=DataInterface() {propertyValues {
          XCurve {keys [<at 0 0>,<at 1 0>]}
          YCurve {keys [<at 0 0>,<at 1 0>]}
          ZCurve {keys [
            {arriveTangent=-342.6536, interpMode=RCIM_Cubic, leaveTangent=-342.6536, tangentMode=RCTM_User, time=0, value=0}
            {arriveTangent=-1163.7987, interpMode=RCIM_Cubic, leaveTangent=-1163.7987, tangentMode=RCTM_User, time=1, value=-700}
           ]}
         }})
      )
     )
    ))
   &AccelerationForce(Acceleration=&MakeCustomVectorFromBool(
     Bool=&SetBoolByFloatComparison(A=&VectorLength(Vector=$User.OwnerLocalVelocity), B=300)
     True Vector=<0 0 0>
     False Vector=&MaskVector(
      Mask Source=0
      Mask Parameter=?
      Mask=0
      Masked Vector=<0 -5 0.1>
      Passthrough Vector=&MaskVector(
       Mask Source=0
       Mask Parameter=?
       Mask=1
       Masked Vector=<0 -5 0.1>
       Passthrough Vector=&MaskVector(Mask Source=0, Mask Parameter=?, Mask=2, Masked Vector=<17 -7 3>, Passthrough Vector=<-10 -7 2>)
      )
     )
    ))
   &Drag(
    Drag=&MaskFloat(
     Mask Source=0
     Mask Parameter=?
     Mask=0
     Masked Float=3
     Passthrough Float=&MaskFloat(
      Mask Source=0
      Mask Parameter=?
      Mask=1
      Masked Float=3
      Passthrough Float=&MaskFloat(Mask Source=0, Mask Parameter=?, Mask=2, Masked Float=0.5, Passthrough Float=0.5)
     )
    )
    Use Linear Drag=true
    Use Rotational Drag=true
   )
   &SolveForcesAndVelocity(Clamp Velocity=false, Limit Acceleration=false)
   &ScaleSpriteSizeBySpeed(Max Scale Factor=<1.25 10>, Min Scale Factor=<0.5 1>, Velocity Threshold=30)
   &ScaleColor(
    Scale Alpha=&FloatFromCurve(
     FloatCurve=DataInterface() {propertyValues {Curve {
        defaultValue=1
        keys [{interpMode=RCIM_Cubic, tangentMode=RCTM_User, time=0, value=1}, {interpMode=RCIM_Cubic, tangentMode=RCTM_User, time=1, value=0}]
       }}}
     Scale Curve=&FN_FrontEndFloatModifier(Gameplay=1, Front End=0.4)
    )
    ScaleA=true
    ScaleRGB=true
   )
  ]
  rendererClasses [&NiagaraSpriteRendererProperties]
  renderers [{rendererClass=&NiagaraSpriteRendererProperties}]
 }
 New_Splashes {
  emitterUpdateScript [
   &EmitterState()
   &SpawnBurst_Instantaneous(
    Spawn Count=5
    Spawn Time=&FN_FrontEndFloatModifier(
     Gameplay=&MakeCustomFloatFromBool(
      Bool=&SetBoolByFloatComparison(A=&VectorLength(Vector=$User.OwnerLocalVelocity), B=20)
      True Float=1.5
      False Float=1.1
     )
     Front End=1.5
    )
    Use Spawn Probability=false
    UseLoopCountLimit=false
   ) {module=SpawnBurst_Instantaneous001}
   &SpawnBurst_Instantaneous(
    Spawn Count=5
    Spawn Group=1
    Spawn Time=&FN_FrontEndFloatModifier(
     Gameplay=&MakeCustomFloatFromBool(
      Bool=&SetBoolByFloatComparison(A=&VectorLength(Vector=$User.OwnerLocalVelocity), B=20)
      True Float=3.4
      False Float=3
     )
     Front End=3.4
    )
    Use Spawn Probability=false
    UseLoopCountLimit=false
   ) {module=SpawnBurst_Instantaneous002}
   &SpawnBurst_Instantaneous(
    Spawn Count=5
    Spawn Time=&FN_FrontEndFloatModifier(
     Gameplay=&MakeCustomFloatFromBool(
      Bool=&SetBoolByFloatComparison(A=&VectorLength(Vector=$User.OwnerLocalVelocity), B=20)
      True Float=5.2
      False Float=4.8
     )
     Front End=5.2
    )
    Use Spawn Probability=false
    UseLoopCountLimit=false
   ) {module=SpawnBurst_Instantaneous008}
   &SpawnBurst_Instantaneous(
    Spawn Count=5
    Spawn Group=1
    Spawn Time=&FN_FrontEndFloatModifier(
     Gameplay=&MakeCustomFloatFromBool(
      Bool=&SetBoolByFloatComparison(A=&VectorLength(Vector=$User.OwnerLocalVelocity), B=20)
      True Float=7.1
      False Float=6.7
     )
     Front End=7.1
    )
    Use Spawn Probability=false
    UseLoopCountLimit=false
   ) {module=SpawnBurst_Instantaneous007}
   &SpawnBurst_Instantaneous(
    Spawn Count=5
    Spawn Time=&FN_FrontEndFloatModifier(
     Gameplay=&MakeCustomFloatFromBool(
      Bool=&SetBoolByFloatComparison(A=&VectorLength(Vector=$User.OwnerLocalVelocity), B=20)
      True Float=9
      False Float=8.6
     )
     Front End=9
    )
    Use Spawn Probability=false
    UseLoopCountLimit=false
   ) {module=SpawnBurst_Instantaneous005}
   &SpawnBurst_Instantaneous(
    Spawn Count=5
    Spawn Group=1
    Spawn Time=&FN_FrontEndFloatModifier(
     Gameplay=&MakeCustomFloatFromBool(
      Bool=&SetBoolByFloatComparison(A=&VectorLength(Vector=$User.OwnerLocalVelocity), B=20)
      True Float=10.7
      False Float=10.3
     )
     Front End=10.7
    )
    Use Spawn Probability=false
    UseLoopCountLimit=false
   ) {module=SpawnBurst_Instantaneous006}
   &SpawnBurst_Instantaneous(
    Spawn Count=5
    Spawn Time=&FN_FrontEndFloatModifier(
     Gameplay=&MakeCustomFloatFromBool(
      Bool=&SetBoolByFloatComparison(A=&VectorLength(Vector=$User.OwnerLocalVelocity), B=20)
      True Float=12.6
      False Float=12.2
     )
     Front End=12.6
    )
    Use Spawn Probability=false
    UseLoopCountLimit=false
   ) {module=SpawnBurst_Instantaneous004}
   &SpawnBurst_Instantaneous(
    Spawn Count=5
    Spawn Group=1
    Spawn Time=&FN_FrontEndFloatModifier(
     Gameplay=&MakeCustomFloatFromBool(
      Bool=&SetBoolByFloatComparison(A=&VectorLength(Vector=$User.OwnerLocalVelocity), B=20)
      True Float=14.3
      False Float=13.9
     )
     Front End=14.3
    )
    Use Spawn Probability=false
    UseLoopCountLimit=false
   ) {module=SpawnBurst_Instantaneous003}
  ]
  particleSpawnScript [
   &InitializeParticle(
    Color=&FN_Select_Color_from_Time_Of_Day(
     Morning=&MakeCustomLinearColorFromBool(
      Bool=&SetBoolByFloatComparison(
       A=&FN_PickFloatDependOnPlatform(
        TargetPlatform=DataInterface() {propertyValues {Platforms {
           qualityLevelMask=-29
           deviceProfileStates [
            {profileName=Windows, qualityLevelMask=0, setQualityLevelMask=3}
            {profileName=Mac, qualityLevelMask=0, setQualityLevelMask=3}
            {profileName=Linux, qualityLevelMask=0, setQualityLevelMask=3}
            {profileName=IOS, qualityLevelMask=0, setQualityLevelMask=4294967295}
           ]
          }}}
        TargetPlatformActive=1
        TargetPlatformNoActive=0
       )
       B=0
      )
      True Color=<rgba 0.2198 0.3436 0.3802 1>
      False Color=&SelectLinearColorFromArray(
       Array Sampling Mode=::NewEnumerator1
       Direct Array Index=$Engine.QualityLevel
       Color Selection Array=DataInterface() {propertyValues {ColorData [<rgba 0.4843 0.7743 0.874 0.3>, <rgba 0.318 0.5763 0.674 0.3>]}}
       Random Seed=?
      )
     )
     Day=&MakeCustomLinearColorFromBool(
      Bool=&SetBoolByFloatComparison(
       A=&FN_PickFloatDependOnPlatform(
        TargetPlatform=DataInterface() {propertyValues {Platforms {
           qualityLevelMask=-29
           deviceProfileStates [
            {profileName=Windows, qualityLevelMask=0, setQualityLevelMask=3}
            {profileName=Mac, qualityLevelMask=0, setQualityLevelMask=3}
            {profileName=Linux, qualityLevelMask=0, setQualityLevelMask=3}
            {profileName=IOS, qualityLevelMask=0, setQualityLevelMask=4294967295}
           ]
          }}}
        TargetPlatformActive=1
        TargetPlatformNoActive=0
       )
       B=0
      )
      True Color=<rgba 0.3187 0.4748 0.4896 1>
      False Color=&RandomRangeLinearColor(
       Minimum=&SelectLinearColorFromArray(
        Array Sampling Mode=::NewEnumerator1
        Direct Array Index=$Engine.QualityLevel
        Color Selection Array=DataInterface() {propertyValues {ColorData [<rgba 0.3444 0.5975 0.8073 1>, <rgba 0.0513 0.5095 0.8073 1>]}}
        Random Seed=?
       )
       Maximum=&SelectLinearColorFromArray(
        Array Sampling Mode=::NewEnumerator1
        Direct Array Index=$Engine.QualityLevel
        Color Selection Array=DataInterface() {propertyValues {ColorData [<rgba 0.323 0.6781 0.5715 1>, <rgba 0.0939 0.6781 0.5875 1>]}}
        Random Seed=?
       )
      )
     )
     Evening=&MakeCustomLinearColorFromBool(
      Bool=&SetBoolByFloatComparison(
       A=&FN_PickFloatDependOnPlatform(
        TargetPlatform=DataInterface() {propertyValues {Platforms {
           qualityLevelMask=-29
           deviceProfileStates [
            {profileName=Windows, qualityLevelMask=0, setQualityLevelMask=3}
            {profileName=Mac, qualityLevelMask=0, setQualityLevelMask=3}
            {profileName=Linux, qualityLevelMask=0, setQualityLevelMask=3}
            {profileName=IOS, qualityLevelMask=0, setQualityLevelMask=4294967295}
           ]
          }}}
        TargetPlatformActive=1
        TargetPlatformNoActive=0
       )
       B=0
      )
      True Color=<rgba 0.2198 0.3436 0.3802 1>
      False Color=&SelectLinearColorFromArray(
       Array Sampling Mode=::NewEnumerator1
       Direct Array Index=$Engine.QualityLevel
       Color Selection Array=DataInterface() {propertyValues {ColorData [<rgba 0.4164 0.8865 0.8712 1>, <rgba 0.19 0.9943 1 1>]}}
       Random Seed=?
      )
     )
     Night=&MakeCustomLinearColorFromBool(
      Bool=&SetBoolByFloatComparison(
       A=&FN_PickFloatDependOnPlatform(
        TargetPlatform=DataInterface() {propertyValues {Platforms {
           qualityLevelMask=-29
           deviceProfileStates [
            {profileName=Windows, qualityLevelMask=0, setQualityLevelMask=3}
            {profileName=Mac, qualityLevelMask=0, setQualityLevelMask=3}
            {profileName=Linux, qualityLevelMask=0, setQualityLevelMask=3}
            {profileName=IOS, qualityLevelMask=0, setQualityLevelMask=4294967295}
           ]
          }}}
        TargetPlatformActive=1
        TargetPlatformNoActive=0
       )
       B=0
      )
      True Color=<rgba 0.2198 0.3436 0.3802 0.6>
      False Color=&SelectLinearColorFromArray(
       Array Sampling Mode=::NewEnumerator1
       Direct Array Index=$Engine.QualityLevel
       Color Selection Array=DataInterface() {propertyValues {ColorData [<rgba 0.3753 0.6794 0.7948 0.6>, <rgba 0.0166 0.6664 1 0.6>]}}
       Random Seed=?
      )
     )
    )
    Lifetime=&MakeCustomFloatFromBool(
     Bool=&SetBoolByFloatComparison(A=&VectorLength(Vector=$User.OwnerLocalVelocity), B=300)
     True Float=0.7
     False Float=1
    )
    Sprite Size Mode=::NewEnumerator1
    Uniform Sprite Size Max=&MakeCustomFloatFromBool(
     Bool=&SetBoolByFloatComparison(A=&VectorLength(Vector=$User.OwnerLocalVelocity), B=20)
     True Float=70
     False Float=&FN_FrontEndFloatModifier(Gameplay=50, Front End=55)
    )
    Uniform Sprite Size Min=&MakeCustomFloatFromBool(
     Bool=&SetBoolByFloatComparison(A=&VectorLength(Vector=$User.OwnerLocalVelocity), B=20)
     True Float=22
     False Float=&FN_FrontEndFloatModifier(Gameplay=15, Front End=30)
    )
    UsePositionOffset=false
    Write Material Random=true
   )
   &SubUVAnimation(
    "Sprite Renderer "=DataInterface() {propertyValues {SpriteRenderer=&self:New_Splashes_0.NiagaraSpriteRendererProperties_1}}
    SubUV Animation Mode:ENiagara_SubUVLookupModeV2=::NewEnumerator2
    UseEndFrame=false
    UseStartFrame=false
   )
   &ShapeLocation(
    Handle Radius:NiagaraFloat=&MakeCustomFloatFromBool(
     Bool=&SetBoolByFloatComparison(A=&VectorLength(Vector=$User.OwnerLocalVelocity), B=20)
     True Float=10
     False Float=11
    )
    Large Radius=&MakeCustomFloatFromBool(
     Bool=&SetBoolByFloatComparison(A=&VectorLength(Vector=$User.OwnerLocalVelocity), B=20)
     True Float=11
     False Float=12
    )
    MaskSpawnGroup=false
    Non Uniform Scale=<1.7 3 0.85>
    Override Seed=false
    Shape Primitive=::NewEnumerator3
    UseOwnerScale=true
   ) {module=ShapeLocation001}
   &AddVelocity(Velocity=&MakeCustomVectorFromBool(
     Bool=&SetBoolByFloatComparison(A=&VectorLength(Vector=$User.OwnerLocalVelocity), B=300)
     True Vector=<0 0 10>
     False Vector=<0 0 15>
    )) {module=AddVelocity001}
   &AddVelocity(
    Module.Use Falloff Distance=false
    Origin Offset=<-10 0 0>
    Override Seed=false
    Velocity Mode=::NewEnumerator1
    Velocity Speed=&MakeCustomFloatFromBool(
     Bool=&SetBoolByFloatComparison(A=&VectorLength(Vector=$User.OwnerLocalVelocity), B=300)
     True Float=0
     False Float=&Multiply_Float(A=&RandomRangeFloat(Minimum=25, Maximum=35), B=$User.Intensity)
    )
   )
   &CameraOffset(Camera Offset Amount=&RandomRangeFloat(
     Minimum=3
     Maximum=&MakeCustomFloatFromBool(
      Bool=&SetBoolByFloatComparison(A=&VectorLength(Vector=$User.OwnerLocalVelocity), B=20)
      True Float=10
      False Float=6
     )
    ))
   &SkeletalMeshLocation(
    MaskSpawnGroup=true
    Mesh Sampling Type=::NewEnumerator5
    Sampled Position Offset=&FN_FrontEndVector3Modifier(
     Gameplay=&MakeCustomVectorFromBool(
      Bool=&SetBoolByFloatComparison(A=&VectorLength(Vector=$User.OwnerLocalVelocity), B=20)
      True Vector=&MakeCustomVectorFromBool(
       Bool=&SetBoolByFloatComparison(A=&VectorLength(Vector=$User.OwnerLocalVelocity), B=300)
       True Vector=<0 0 -10>
       False Vector=<2 23 -6>
      )
      False Vector=<2 18 -18>
     )
     Front End=<0 0 -3>
    )
    Sampled Position Offset Coordinate Space:ENiagaraSkelMeshTransforms=::NewEnumerator2
    Skeletal Mesh=$System.Left Foot
    Socket Sampling Mode=::NewEnumerator5
   )
   &SkeletalMeshLocation(
    MaskSpawnGroup=true
    Mesh Sampling Type=::NewEnumerator5
    Sampled Position Offset=&FN_FrontEndVector3Modifier(
     Gameplay=&MakeCustomVectorFromBool(
      Bool=&SetBoolByFloatComparison(A=&VectorLength(Vector=$User.OwnerLocalVelocity), B=20)
      True Vector=&MakeCustomVectorFromBool(
       Bool=&SetBoolByFloatComparison(A=&VectorLength(Vector=$User.OwnerLocalVelocity), B=300)
       True Vector=<0 0 -10>
       False Vector=<-2 21 -5>
      )
      False Vector=<-2 18 -18>
     )
     Front End=<0 0 -3>
    )
    Sampled Position Offset Coordinate Space:ENiagaraSkelMeshTransforms=::NewEnumerator2
    Skeletal Mesh=$System.Right Foot
    Socket Sampling Mode=::NewEnumerator5
    Spawn Group Mask=1
   ) {module=SkeletalMeshLocation001}
  ]
  particleUpdateScript [
   &ParticleState()
   &GravityForce(Gravity=&MakeCustomVectorFromBool(
     Bool=&SetBoolByFloatComparison(A=&VectorLength(Vector=$User.OwnerLocalVelocity), B=20)
     True Vector=&MakeCustomVectorFromBool(
      Bool=&SetBoolByFloatComparison(A=&VectorLength(Vector=$User.OwnerLocalVelocity), B=300)
      True Vector=&VectorFromCurve(
       VectorCurve=DataInterface() {propertyValues {
         XCurve {keys [<at 0 0>,<at 1 0>]}
         YCurve {keys [
           {arriveTangent=-1.6862, interpMode=RCIM_Cubic, leaveTangent=-1.6862, tangentMode=RCTM_User, time=0, value=0}
           {interpMode=RCIM_Cubic, tangentMode=RCTM_User, time=1, value=0}
          ]}
         ZCurve {keys [
           {arriveTangent=-2016.3778, interpMode=RCIM_Cubic, leaveTangent=-2016.3778, tangentMode=RCTM_User, time=0, value=0}
           {arriveTangent=54.3671, interpMode=RCIM_Cubic, leaveTangent=54.3671, tangentMode=RCTM_User, time=1, value=-600}
          ]}
        }}
       Scale Curve=<0.3 0.3 0.2>
      )
      False Vector=&VectorFromCurve(
       VectorCurve=DataInterface() {propertyValues {
         XCurve {keys [<at 0 0>,<at 1 0>]}
         YCurve {keys [
           {arriveTangent=-21.7461, interpMode=RCIM_Cubic, leaveTangent=-21.7461, tangentMode=RCTM_User, time=0, value=-10}
           {interpMode=RCIM_Cubic, tangentMode=RCTM_User, time=1, value=-20}
          ]}
         ZCurve {keys [
           {arriveTangent=7034.374, interpMode=RCIM_Cubic, leaveTangent=7034.374, tangentMode=RCTM_User, time=0, value=0}
           {arriveTangent=-2079.8411, interpMode=RCIM_Cubic, leaveTangent=-2079.8411, tangentMode=RCTM_User, time=0.06, value=0}
           {arriveTangent=54.3671, interpMode=RCIM_Cubic, leaveTangent=54.3671, tangentMode=RCTM_User, time=1, value=-600}
          ]}
        }}
       Scale Curve=<0.3 0.3 0.2>
      )
     )
     False Vector=&VectorFromCurve(
      VectorCurve=DataInterface() {propertyValues {
        XCurve {keys [<at 0 0>,<at 1 0>]}
        YCurve {keys [
          {arriveTangent=-21.7461, interpMode=RCIM_Cubic, leaveTangent=-21.7461, tangentMode=RCTM_User, time=0, value=-10}
          {interpMode=RCIM_Cubic, tangentMode=RCTM_User, time=1, value=-20}
         ]}
        ZCurve {keys [
          {arriveTangent=3327.699, interpMode=RCIM_Cubic, leaveTangent=3327.699, tangentMode=RCTM_User, time=0, value=0}
          {arriveTangent=-2079.8411, interpMode=RCIM_Cubic, leaveTangent=-2079.8411, tangentMode=RCTM_User, time=0.06, value=0}
          {arriveTangent=54.3671, interpMode=RCIM_Cubic, leaveTangent=54.3671, tangentMode=RCTM_User, time=1, value=-600}
         ]}
       }}
      Scale Curve=<0.3 0.3 0.2>
     )
    ))
   &BlendSpace("Position Blend (World <-> Local)":NiagaraFloat=&MakeCustomFloatFromBool(
     Bool=&SetBoolByFloatComparison(A=&VectorLength(Vector=$Engine.Owner.Velocity), B=300)
     True Float=1
     False Float=0
     Translation Contribution:NiagaraFloat=1
     Rotation Contribution:NiagaraFloat=1
    )) {module=BlendSpace001}
   &ScaleSpriteSize(
    Uniform Curve Scale=0.7
    Uniform Curve Sprite Scale=DataInterface() {propertyValues {Curve {keys [
        {arriveTangent=5.6756, interpMode=RCIM_Cubic, leaveTangent=5.6756, tangentMode=RCTM_User, time=0, value=0.15}
        {arriveTangent=1.1179, interpMode=RCIM_Cubic, leaveTangent=1.1179, tangentMode=RCTM_User, time=0.2, value=0.7}
        {arriveTangent=0.0224, interpMode=RCIM_Cubic, leaveTangent=0.0224, tangentMode=RCTM_User, time=1, value=1}
       ]}}}
   )
   &ScaleColor(
    Scale Alpha=&FloatFromCurve(
     FloatCurve=DataInterface() {propertyValues {Curve {keys [
         {arriveTangent=22.5928, interpMode=RCIM_Cubic, leaveTangent=22.5928, tangentMode=RCTM_User, time=0, value=0.3}
         {arriveTangent=8.0021, interpMode=RCIM_Cubic, leaveTangent=8.0021, tangentMode=RCTM_User, time=0.02, value=0.95}
         {arriveTangent=-2.7573, interpMode=RCIM_Cubic, leaveTangent=-3.1481, tangentMode=RCTM_Break, time=0.2, value=0.95}
         {arriveTangentWeight=0.2045, interpMode=RCIM_Cubic, tangentMode=RCTM_User, time=1, value=0}
        ]}}}
     Scale Curve=0.7
    )
    ScaleA=true
    ScaleRGB=true
   )
   &SolveForcesAndVelocity(Clamp Velocity=false, Limit Acceleration=false)
   &DynamicMaterialParameters(
    Index 0 Param 1=&FN_FrontEndFloatModifier(
     Gameplay=&FloatFromCurve(FloatCurve=DataInterface() {propertyValues {Curve {keys [
          {arriveTangent=-0.0467, interpMode=RCIM_Cubic, leaveTangent=-0.0467, tangentMode=RCTM_User, time=0.1, value=1}
          {arriveTangent=-4.9722, tangentMode=RCTM_User, time=1, value=0}
         ]}}})
     Front End=&FloatFromCurve(FloatCurve=DataInterface() {propertyValues {Curve {keys [
          {arriveTangent=-0.0467, interpMode=RCIM_Cubic, leaveTangent=-0.0467, tangentMode=RCTM_User, time=0.1, value=1}
          {arriveTangent=-2.0505, interpMode=RCIM_Cubic, leaveTangent=-2.0505, tangentMode=RCTM_User, time=0.7044, value=0.6698}
          {arriveTangent=0.0025, tangentMode=RCTM_User, time=1, value=0}
         ]}}})
    )
    Module.Index0Float0Write=true
    Module.Index0Float1Write=true
    Module.Index0Float2Write=true
    Module.Index0Float3Write=true
   )
  ]
  rendererClasses [&NiagaraSpriteRendererProperties]
  renderers [{rendererClass=&NiagaraSpriteRendererProperties}]
 }
 Sand {
  emitterUpdateScript [
   &EmitterState()
   &SpawnBurst_Instantaneous(
    Spawn Count=2
    Spawn Time=&FN_FrontEndFloatModifier(
     Gameplay=&MakeCustomFloatFromBool(
      Bool=&SetBoolByFloatComparison(A=&VectorLength(Vector=$User.OwnerLocalVelocity), B=20)
      True Float=1.5
      False Float=1.1
     )
     Front End=1.5
    )
    Use Spawn Probability=false
    UseLoopCountLimit=false
   ) {module=SpawnBurst_Instantaneous001}
   &SpawnBurst_Instantaneous(
    Spawn Count=2
    Spawn Group=1
    Spawn Time=&FN_FrontEndFloatModifier(
     Gameplay=&MakeCustomFloatFromBool(
      Bool=&SetBoolByFloatComparison(A=&VectorLength(Vector=$User.OwnerLocalVelocity), B=20)
      True Float=3.4
      False Float=3
     )
     Front End=3.4
    )
    Use Spawn Probability=false
    UseLoopCountLimit=false
   ) {module=SpawnBurst_Instantaneous002}
   &SpawnBurst_Instantaneous(
    Spawn Count=2
    Spawn Time=&FN_FrontEndFloatModifier(
     Gameplay=&MakeCustomFloatFromBool(
      Bool=&SetBoolByFloatComparison(A=&VectorLength(Vector=$User.OwnerLocalVelocity), B=20)
      True Float=5.2
      False Float=4.8
     )
     Front End=5.2
    )
    Use Spawn Probability=false
    UseLoopCountLimit=false
   ) {module=SpawnBurst_Instantaneous008}
   &SpawnBurst_Instantaneous(
    Spawn Count=2
    Spawn Group=1
    Spawn Time=&FN_FrontEndFloatModifier(
     Gameplay=&MakeCustomFloatFromBool(
      Bool=&SetBoolByFloatComparison(A=&VectorLength(Vector=$User.OwnerLocalVelocity), B=20)
      True Float=7.1
      False Float=6.7
     )
     Front End=7.1
    )
    Use Spawn Probability=false
    UseLoopCountLimit=false
   ) {module=SpawnBurst_Instantaneous007}
   &SpawnBurst_Instantaneous(
    Spawn Count=2
    Spawn Time=&FN_FrontEndFloatModifier(
     Gameplay=&MakeCustomFloatFromBool(
      Bool=&SetBoolByFloatComparison(A=&VectorLength(Vector=$User.OwnerLocalVelocity), B=20)
      True Float=9
      False Float=8.6
     )
     Front End=9
    )
    Use Spawn Probability=false
    UseLoopCountLimit=false
   ) {module=SpawnBurst_Instantaneous005}
   &SpawnBurst_Instantaneous(
    Spawn Count=2
    Spawn Group=1
    Spawn Time=&FN_FrontEndFloatModifier(
     Gameplay=&MakeCustomFloatFromBool(
      Bool=&SetBoolByFloatComparison(A=&VectorLength(Vector=$User.OwnerLocalVelocity), B=20)
      True Float=10.7
      False Float=10.3
     )
     Front End=10.7
    )
    Use Spawn Probability=false
    UseLoopCountLimit=false
   ) {module=SpawnBurst_Instantaneous006}
   &SpawnBurst_Instantaneous(
    Spawn Count=2
    Spawn Time=&FN_FrontEndFloatModifier(
     Gameplay=&MakeCustomFloatFromBool(
      Bool=&SetBoolByFloatComparison(A=&VectorLength(Vector=$User.OwnerLocalVelocity), B=20)
      True Float=12.6
      False Float=12.2
     )
     Front End=12.6
    )
    Use Spawn Probability=false
    UseLoopCountLimit=false
   ) {module=SpawnBurst_Instantaneous004}
   &SpawnBurst_Instantaneous(
    Spawn Count=2
    Spawn Group=1
    Spawn Time=&FN_FrontEndFloatModifier(
     Gameplay=&MakeCustomFloatFromBool(
      Bool=&SetBoolByFloatComparison(A=&VectorLength(Vector=$User.OwnerLocalVelocity), B=20)
      True Float=14.3
      False Float=13.9
     )
     Front End=14.3
    )
    Use Spawn Probability=false
    UseLoopCountLimit=false
   ) {module=SpawnBurst_Instantaneous003}
  ]
  particleSpawnScript [
   &InitializeParticle(
    Color Maximum=<rgba 0.674 0.4176 0.2984 1.5>
    Color Minimum=&Multiply_LinearColorByFloat(
     Linear Color=<rgba 0.8406 0.4865 0.2881 1.5>
     Float=&FNTimeOfDayScalar(Night Scale=0.5, Morning Scale=0.6, Evening Scale=0.6)
    )
    Color Mode=::NewEnumerator2
    Lifetime=1.5
    Sprite Size Mode=::NewEnumerator1
    Uniform Sprite Size Max=&MakeCustomFloatFromBool(
     Bool=&SetBoolByFloatComparison(A=&VectorLength(Vector=$User.OwnerLocalVelocity), B=20)
     True Float=120
     False Float=85
    )
    Uniform Sprite Size Min=&MakeCustomFloatFromBool(
     Bool=&SetBoolByFloatComparison(A=&VectorLength(Vector=$User.OwnerLocalVelocity), B=20)
     True Float=70
     False Float=45
    )
    UsePositionOffset=false
    Write Material Random=true
   )
   &ShapeLocation(
    Handle Radius=9
    Large Radius=11
    MaskSpawnGroup=false
    Non Uniform Scale=<1.7 4 0.85>
    Override Seed=false
    Shape Primitive=::NewEnumerator3
    UseOwnerScale=true
   ) {module=ShapeLocation001}
   &AddVelocity(
    Cone Angle=30
    Cone Axis=<0 -1 0>
    Module.UseConeFalloff=false
    Module.UseVelDistribution=false
    Override Seed=false
    Velocity Mode=::NewEnumerator2
    Velocity Speed=&Multiply_Float(A=&RandomRangeFloat(Minimum=10, Maximum=15), B=$User.Intensity)
   ) {module=AddVelocity001}
   &AddVelocity(
    Module.Use Falloff Distance=false
    Origin Offset=<-10 0 0>
    Override Seed=false
    Velocity Mode=::NewEnumerator1
    Velocity Speed=&Multiply_Float(A=&RandomRangeFloat(Minimum=15, Maximum=20), B=$User.Intensity)
   )
   &CameraOffset(Camera Offset Amount=&RandomRangeFloat(Minimum=5, Maximum=10))
   &SkeletalMeshLocation(
    MaskSpawnGroup=true
    Mesh Sampling Type=::NewEnumerator5
    Sampled Position Offset=&FN_FrontEndVector3Modifier(Gameplay=<0 12 -5>, Front End=<0 -6 -3>)
    Sampled Position Offset Coordinate Space:ENiagaraSkelMeshTransforms=::NewEnumerator2
    Skeletal Mesh=$System.Left Foot
    Socket Sampling Mode=::NewEnumerator5
   )
   &SkeletalMeshLocation(
    MaskSpawnGroup=true
    Mesh Sampling Type=::NewEnumerator5
    Sampled Position Offset=&FN_FrontEndVector3Modifier(Gameplay=<0 12 -5>, Front End=<0 -6 -3>)
    Sampled Position Offset Coordinate Space:ENiagaraSkelMeshTransforms=::NewEnumerator2
    Skeletal Mesh=$System.Right Foot
    Socket Sampling Mode=::NewEnumerator5
    Spawn Group Mask=1
   ) {module=SkeletalMeshLocation001}
  ]
  particleUpdateScript [
   &ParticleState()
   &GravityForce(Gravity=&MakeCustomVectorFromBool(
     Bool=&SetBoolByFloatComparison(A=&VectorLength(Vector=$User.OwnerLocalVelocity), B=20)
     True Vector=&VectorFromCurve(VectorCurve=DataInterface() {propertyValues {
        XCurve {keys [<at 0 0>,<at 1 0>]}
        YCurve {keys [
          {arriveTangent=-21.7461, interpMode=RCIM_Cubic, leaveTangent=-21.7461, tangentMode=RCTM_User, time=0, value=-10}
          {interpMode=RCIM_Cubic, tangentMode=RCTM_User, time=1, value=-20}
         ]}
        ZCurve {keys [
          {arriveTangent=578.5428, interpMode=RCIM_Cubic, leaveTangent=578.5428, tangentMode=RCTM_User, time=0, value=0}
          {arriveTangent=-6573.3198, interpMode=RCIM_Cubic, leaveTangent=-2867.7527, tangentMode=RCTM_Break, time=0.2, value=0}
          {arriveTangent=54.3671, interpMode=RCIM_Cubic, leaveTangent=54.3671, tangentMode=RCTM_User, time=1, value=-600}
         ]}
       }})
     False Vector=&VectorFromCurve(VectorCurve=DataInterface() {propertyValues {
        XCurve {keys [<at 0 0>,<at 1 0>]}
        YCurve {keys [
          {arriveTangent=-21.7461, interpMode=RCIM_Cubic, leaveTangent=-21.7461, tangentMode=RCTM_User, time=0, value=-10}
          {interpMode=RCIM_Cubic, tangentMode=RCTM_User, time=1, value=-20}
         ]}
        ZCurve {keys [
          {arriveTangent=800.1272, interpMode=RCIM_Cubic, leaveTangent=800.1272, tangentMode=RCTM_User, time=0, value=0}
          {arriveTangent=-3507.2607, interpMode=RCIM_Cubic, leaveTangent=-2867.7527, tangentMode=RCTM_Break, time=0.2, value=0}
          {arriveTangent=54.3671, interpMode=RCIM_Cubic, leaveTangent=54.3671, tangentMode=RCTM_User, time=1, value=-600}
         ]}
       }})
    ))
   &BlendSpace("Position Blend (World <-> Local)":NiagaraFloat=&MakeCustomFloatFromBool(
     Bool=&SetBoolByFloatComparison(A=&VectorLength(Vector=$Engine.Owner.Velocity), B=301)
     True Float=1
     False Float=&FloatFromCurve(FloatCurve=DataInterface() {propertyValues {Curve {
         defaultValue=0
         keys [
          {arriveTangent=-7.3261, interpMode=RCIM_Cubic, leaveTangent=-7.3261, tangentMode=RCTM_User, time=0, value=1}
          {arriveTangent=-1.5885, interpMode=RCIM_Cubic, leaveTangent=-1.5885, tangentMode=RCTM_User, time=0.2, value=0.2}
          <at 0.5 0 RCIM_Cubic>
         ]
        }}})
     Translation Contribution:NiagaraFloat=1
     Rotation Contribution:NiagaraFloat=1
    )) {module=BlendSpace001}
   &Drag(Drag=2, Use Linear Drag=true, Use Rotational Drag=false)
   &SolveForcesAndVelocity(Clamp Velocity=false, Limit Acceleration=false) {module=SolveForcesAndVelocity001}
   &ScaleSpriteSize(
    Uniform Curve Scale=0.7
    Uniform Curve Sprite Scale=DataInterface() {propertyValues {Curve {keys [
        {arriveTangent=7.4995, interpMode=RCIM_Cubic, leaveTangent=7.4995, tangentMode=RCTM_User, time=0, value=0.15}
        {arriveTangentWeight=0.1312, interpMode=RCIM_Cubic, leaveTangentWeight=0.3062, tangentMode=RCTM_User, time=0.3, value=1}
        {arriveTangent=0.8571, interpMode=RCIM_Cubic, leaveTangent=0.8571, time=0.8, value=1}
        {arriveTangent=2.6672, arriveTangentWeight=0.266, interpMode=RCIM_Cubic, leaveTangent=2.6672, tangentMode=RCTM_User, time=1, value=1.6}
       ]}}}
   )
   &ScaleColor(
    Scale Alpha=&FloatFromCurve(
     FloatCurve=DataInterface() {propertyValues {Curve {keys [
         <at 0 0.3 RCIM_Cubic>
         {arriveTangent=9.9601, interpMode=RCIM_Cubic, leaveTangent=9.9601, tangentMode=RCTM_User, time=0.02, value=0.95}
         {arriveTangent=-2.7573, interpMode=RCIM_Cubic, leaveTangent=-3.7393, tangentMode=RCTM_Break, time=0.1, value=0.95}
         {arriveTangentWeight=0.2045, interpMode=RCIM_Cubic, tangentMode=RCTM_User, time=1, value=0}
        ]}}}
     Scale Curve=&FN_FrontEndFloatModifier(Gameplay=0.7, Front End=0.4)
    )
    ScaleA=true
    ScaleRGB=true
   )
   &SolveForcesAndVelocity(Clamp Velocity=false, Limit Acceleration=false)
   /SetVariables(Particles.Position=&AddVectorToPosition(
     Position=$Particles.Position
     Vector=&MakeCustomVectorFromBool(
      Bool=&SetBoolByFloatComparison(A=&VectorLength(Vector=$User.OwnerLocalVelocity), B=20)
      True Vector=&VectorFromCurve(
       VectorCurve=DataInterface() {propertyValues {XCurve {keys [<at 0 0>,<at 1 0>]}, YCurve {keys [<at 0 -1>,<at 1 0>]}, ZCurve {keys [<at 0 0>,<at 1 0>]}}}
       Scale Curve=<0.2 0.22 0.2>
      )
      False Vector=<0 0 0>
     )
    )) {isSetParameters=true}
  ]
  rendererClasses [&NiagaraSpriteRendererProperties]
  renderers [{rendererClass=&NiagaraSpriteRendererProperties}]
 }
}
summary {
 systemName=NS_AquaPeony_Splashes
 emitters [
  {emitterName=Sand, rendererClasses [&NiagaraSpriteRendererProperties]}
  {emitterName=New_Splashes, rendererClasses [&NiagaraSpriteRendererProperties]}
  {emitterName=New_HairDroplets, rendererClasses [&NiagaraSpriteRendererProperties]}
 ]
 userVariables [
  User.bIsFrontEnd {type=NiagaraBool}
  User.bIsFrontEndPreview {type=NiagaraBool}
  User.Intensity {type=NiagaraFloat, defaultValue=0.68, description=""}
  User.OwnerLocalVelocity {type=Vector3f, description="", defaultValue {struct=Vector3f, value=<0 0 0>}}
 ]
}
